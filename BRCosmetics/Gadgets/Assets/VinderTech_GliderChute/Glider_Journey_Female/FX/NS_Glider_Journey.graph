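# Niagara effect system for the Journey glider, producing mesh edge spray and sprite speedlines around it.
# Edge spray samples two skeletal mesh sockets by spawn group, mirroring orientation and shaping scale with User RightDot, ForwardDot and system Timeline.
# Front-end modifiers give separate gameplay and menu values, and curves drive dynamic material parameters over particle life.
# Several emitters shown are disabled, including alternate spray, water foam, smoke and one speedline emitter.
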NiagaraSystem /BRCosmetics/Gadgets/Assets/VinderTech_GliderChute/Glider_Journey_Female/FX/NS_Glider_Journey
dependencies {
 usedDataInterfaces [
  &NiagaraDataInterfaceCurve
  &NiagaraDataInterfaceVectorCurve
  &NiagaraDataInterfaceVector2DCurve
 ]
 usedDynamicInputs [&SetBoolByFloatComparison]
 usedModules [
  &SystemState
  &FN_Integrator
  &self:SystemSpawnScript.SystemScriptSource.SystemScriptGraph.NiagaraNodeAssignment_0.SetVariables_#1
  &Timeline
 ]
}
emitters {
 Edge_Spray_Mesh {
  emitterUpdateScript [
   &EmitterState(Inactive Response=::NewEnumerator1, Life Cycle Mode=::NewEnumerator1, Loop Behavior=::NewEnumerator1, Loop Duration=3600, UseLoopDelay=false)
   &SpawnRate(
    SpawnGroup:NiagaraInt32=&RandomRangeInt(Minimum=0, Maximum=1)
    SpawnRate=&FN_FrontEndFloatModifier(Gameplay=90, Front End=60)
    Use Spawn Probability=false
   )
  ]
  particleSpawnScript [
   &FN_SetSpawnGroup(Number of Spawn Groups=2)
   /SetVariables(
    Particles.FrontToBack:NiagaraFloat=&RandomRangeFloat(Minimum=0, Maximum=1)
    Particles.LeanLerp:NiagaraFloat=&Multiply_Float(
     A=$User.RightDot
     B=&Lerp_Float(A=-1, B=1, Alpha=&MakeFloatFromInt(Int=$Particles.SpawnGroup))
    )
   ) {isSetParameters=true, module=SetVariables_#2}
   &InitializeParticle(
    Lifetime Max=1
    Lifetime Min=0.75
    Lifetime Mode=::NewEnumerator1
    Mesh Scale Max=<1 1 1.1>
    Mesh Scale Min=<0.75 0.75 0.75>
    Mesh Scale Mode=::NewEnumerator2
    Sprite Rotation Mode=::NewEnumerator2
    UsePositionOffset=false
    Write Material Random=true
   ) {module=InitializeParticle001}
   &SkeletalMeshLocation(
    Apply Sampled Orientation to Mesh Particles=true
    Bone / Socket Index=$Particles.SpawnGroup
    Inherit Velocity=false
    Interpolate Mesh Sampling=false
    MaskSpawnGroup=false
    Mesh Sampling Type=::NewEnumerator5
    Sampled Position Offset=&Lerp_Vector(
     A=<50 0 0>
     B=&Lerp_Vector(A=<-75 -15 0>, B=<-75 15 1>, Alpha=&MakeFloatFromInt(Int=$Particles.SpawnGroup))
     Alpha=$Particles.FrontToBack
    )
    Sampled Position Offset Coordinate Space:ENiagaraMeshTransforms=::NewEnumerator3
    Skeletal Mesh=$User.SkeletalMesh
    Socket Sampling Mode=::NewEnumerator5
   ) {enabled=false}
   &InitialMeshOrientation(
    Orientation Vector=&Multiply_Vector(
     A=&Add_Vector(
      A=&Lerp_Vector(A=<0 0 0>, B=<-1 0 0>, Alpha=$Particles.FrontToBack)
      B=&Lerp_Vector(A=<-1 0.8 0>, B=<-0.75 1 0>, Alpha=$Particles.LeanLerp)
     )
     B=&Multiply_Vector(
      A=&Lerp_Vector(A=<1 1 1>, B=<1 -1 1>, Alpha=&MakeFloatFromInt(Int=$Particles.SpawnGroup))
      B=&Lerp_Vector(A=<1 1 1>, B=<1 0.75 1>, Alpha=&ClampFloat(Float=$User.ForwardDot))
     )
    )
    Rotation Coordinate Space=::NewEnumerator0
    Use Orientation Vector=true
    Use Rotation Vector=false
   )
   &InitialMeshOrientation(
    Orientation Vector=&Multiply_Vector(
     A=&Add_Vector(
      A=&VectorFromCurve(
       VectorCurve=DataInterface() {propertyValues {
         XCurve {keys [<at 0.33 1 RCIM_Cubic>, {arriveTangent=-1.96, interpMode=RCIM_Cubic, leaveTangent=-1.96, tangentMode=RCTM_User, time=1, value=0.2}]}
         YCurve {keys [<at 0.33 0.5 RCIM_Cubic>, {arriveTangent=-0.6339, interpMode=RCIM_Cubic, leaveTangent=-0.6339, tangentMode=RCTM_User, time=1, value=0.2}]}
         ZCurve {keys [<at 0 0>,<at 1 0>]}
        }}
       CurveIndex=$Particles.FrontToBack
       Scale Curve=<-1 1 1>
      )
      B=&Lerp_Vector(A=<0 0 0>, B=<0.5 0 0>, Alpha=$Particles.LeanLerp)
     )
     B=&Multiply_Vector(
      A=&Lerp_Vector(A=<1 1 1>, B=<1 -1 1>, Alpha=&MakeFloatFromInt(Int=$Particles.SpawnGroup))
      B=&Lerp_Vector(A=<1 1 1>, B=<1 0.4 1>, Alpha=$User.ForwardDot)
     )
    )
    Rotation Coordinate Space=::NewEnumerator0
    Use Orientation Vector=true
    Use Rotation Vector=false
   ) {enabled=false, module=InitialMeshOrientation001}
   /SetVariables(Particles.Scale=&Multiply_Vector(
     A=&VectorFromCurve(
      VectorCurve=DataInterface() {propertyValues {
        XCurve {keys [{arriveTangent=1.7846, interpMode=RCIM_Cubic, leaveTangent=1.7846, tangentMode=RCTM_User, time=0, value=0.4}, <at 1 1 RCIM_Cubic>]}
        YCurve {keys [{arriveTangent=1.7846, interpMode=RCIM_Cubic, leaveTangent=1.7846, tangentMode=RCTM_User, time=0, value=0.4}, <at 1 1 RCIM_Cubic>]}
        ZCurve {keys [{arriveTangent=1.7846, interpMode=RCIM_Cubic, leaveTangent=1.7846, tangentMode=RCTM_User, time=0, value=0.4}, <at 1 1 RCIM_Cubic>]}
       }}
      CurveIndex=$Particles.FrontToBack
      Scale Curve=&Multiply_VectorByFloat(
       Vector=$Particles.Scale
       Float=&Lerp_Float(
        A=&FN_FrontEndFloatModifier(Gameplay=0, Front End=0.4)
        B=0.6
        Alpha=&ClampFloat(Float=$System.Timeline.Time)
       )
      )
     )
     B=&Lerp_Vector(A=<1.5 1 0.2>, B=<1.5 1 1>, Alpha=$Particles.LeanLerp)
    )) {isSetParameters=true, module=SetVariables_#3}
   /SetVariables(Particles.Scale=&Multiply_Vector(
     A=&VectorFromFloat(Value=&FloatFromCurve(
       FloatCurve=DataInterface() {propertyValues {Curve {keys [{interpMode=RCIM_Cubic, tangentMode=RCTM_User, time=0, value=0.4}, {interpMode=RCIM_Cubic, tangentMode=RCTM_User, time=1, value=1}]}}}
       CurveIndex=$Particles.FrontToBack
       Scale Curve=&Multiply_Float(A=1, B=&Lerp_Float(A=1, B=1.4, Alpha=$User.ForwardDot))
      ))
     B=&Lerp_Vector(A=<0.6 0.6 0.6>, B=<0.8 0.8 0.8>, Alpha=$Particles.LeanLerp)
    )) {enabled=false, isSetParameters=true, module=SetVariables_#4}
  ]
  particleUpdateScript [
   &ParticleState()
   /SetVariables(Particles.Scale=&Multiply_Vector(
     A=$Particles.Initial.Scale
     B=&VectorFromCurve(VectorCurve=DataInterface() {propertyValues {
        XCurve {keys [
          {arriveTangent=4.6994, interpMode=RCIM_Cubic, leaveTangent=4.6994, tangentMode=RCTM_User, time=0, value=0.0795}
          {arriveTangent=0.4589, interpMode=RCIM_Cubic, leaveTangent=0.4589, tangentMode=RCTM_User, time=0.1447, value=0.7609}
          <at 1 1>
         ]}
        YCurve {keys [{arriveTangent=2.5596, interpMode=RCIM_Cubic, leaveTangent=2.5596, tangentMode=RCTM_User, time=0, value=0.0767}, <at 1 1>]}
        ZCurve {keys [
          <at 0 0 RCIM_Cubic>
          {arriveTangent=-2.3619, interpMode=RCIM_Cubic, leaveTangent=-2.3619, tangentMode=RCTM_User, time=0.336, value=0.9567}
          <at 0.9988 0>
         ]}
       }})
    )) {isSetParameters=true}
   &SkeletalMeshLocation(
    Apply Sampled Orientation to Mesh Particles=true
    Bone / Socket Index=$Particles.SpawnGroup
    Inherit Velocity=false
    MaskSpawnGroup=false
    Mesh Sampling Type=::NewEnumerator5
    Sampled Position Offset=&Add_Vector(
     A=&Add_Vector(
      A=<-80 0 10>
      B=&Lerp_Vector(A=<0 5 -8>, B=<0 -10 -8>, Alpha=&MakeFloatFromInt(Int=$Particles.SpawnGroup))
     )
     B=&VectorFromCurve(
      VectorCurve=DataInterface() {propertyValues {
        XCurve {keys [<at 0 0>, {arriveTangent=-382.5781, interpMode=RCIM_Cubic, leaveTangent=-382.5781, tangentMode=RCTM_User, time=1, value=-125}]}
        YCurve {keys [<at 0.33 0 RCIM_Cubic>, {arriveTangent=-39.3211, interpMode=RCIM_Cubic, leaveTangent=-39.3211, tangentMode=RCTM_User, time=1, value=10}]}
        ZCurve {keys [<at 0 0>,<at 1 10>]}
       }}
      CurveIndex=$Particles.FrontToBack
      Scale Curve=&Lerp_Vector(A=<0.2 1 1>, B=<0.5 0 1>, Alpha=&MakeFloatFromInt(Int=$Particles.SpawnGroup))
     )
    )
    Sampled Position Offset Coordinate Space:ENiagaraMeshTransforms=::NewEnumerator3
    Skeletal Mesh=$User.SkeletalMesh
    Socket Sampling Mode=::NewEnumerator5
   ) {module=SkeletalMeshLocation001}
   &SolveForcesAndVelocity(Clamp Velocity=false, Limit Acceleration=false)
   &DynamicMaterialParameters(
    Index 0 Param 1=&FloatFromCurve(FloatCurve=DataInterface() {propertyValues {Curve {keys [
         {arriveTangent=1.2629, interpMode=RCIM_Cubic, leaveTangent=1.2629, tangentMode=RCTM_User, time=-0.0045, value=-1.9799}
         {arriveTangent=4.4591, interpMode=RCIM_Cubic, leaveTangent=4.4591, tangentMode=RCTM_User, time=1, value=0}
        ]}}})
    Index 0 Param 2=&FloatFromCurve(FloatCurve=DataInterface() {propertyValues {Curve {keys [<at 0.0102 4>,<at 0.9749 0>]}}})
    Index 0 Param 3=&MakeFloatFromVector(Vector=$Particles.Scale, Channel=::NewEnumerator0)
    Index 0 Param 4=1
    Index 1 Param 1=&FloatFromCurve(FloatCurve=DataInterface() {propertyValues {Curve {keys [<at 0.7 0 RCIM_Cubic>, {arriveTangent=6.4641, interpMode=RCIM_Cubic, leaveTangent=6.4641, tangentMode=RCTM_User, time=1, value=1}]}}})
    Module.Param0WriteEnabled=true
    Module.Param0WriteEnabled1=true
    Module.Param1WriteEnabled=true
    Module.Param1WriteEnabled1=true
    Module.Param2WriteEnabled=true
    Module.Param2WriteEnabled1=true
    Module.Param3WriteEnabled=true
    Module.Param3WriteEnabled1=true
    Write Parameter Index 1=true
   )
  ]
  rendererClasses [&NiagaraMeshRendererProperties]
  renderers [{rendererClass=&NiagaraMeshRendererProperties, rendererIndex=0}]
 }
 Edge_Spray_Mesh001 {
  bEnabled=false
  emitterUpdateScript [
   &EmitterState(Inactive Response=::NewEnumerator1, Life Cycle Mode=::NewEnumerator1, Loop Behavior=::NewEnumerator1, Loop Duration=3600, UseLoopDelay=false)
   &SpawnRate(
    SpawnGroup:NiagaraInt32=&RandomRangeInt(Minimum=0, Maximum=1)
    SpawnRate=&FN_FrontEndFloatModifier(Gameplay=30, Front End=30)
    Use Spawn Probability=false
   )
  ]
  particleSpawnScript [
   &FN_SetSpawnGroup(Number of Spawn Groups=2)
   /SetVariables(
    Particles.FrontToBack:NiagaraFloat=&RandomRangeFloat(Minimum=0, Maximum=1)
    Particles.LeanLerp:NiagaraFloat=&Multiply_Float(
     A=$User.RightDot
     B=&Lerp_Float(A=-1, B=1, Alpha=&MakeFloatFromInt(Int=$Particles.SpawnGroup))
    )
   ) {isSetParameters=true, module=SetVariables_#2}
   &InitializeParticle(
    Lifetime Max=1
    Lifetime Min=0.75
    Lifetime Mode=::NewEnumerator1
    Mesh Scale Max=<1 1 1.1>
    Mesh Scale Min=<0.75 0.75 0.75>
    Mesh Scale Mode=::NewEnumerator2
    Sprite Rotation Mode=::NewEnumerator2
    UsePositionOffset=false
    Write Material Random=true
   ) {enabled=false, module=InitializeParticle001}
   /Niagara/Modules/Spawn/Initialization/InitializeParticle(
    Lifetime=1.5
    Mesh Scale=&UniformRangedVector(Minimum=<1.25 1.25 0.8>, Maximum=<2 1.4 1.15>)
    Module.Write Color=true
    Module.Write Lifetime=true
    Module.Write Mass=false
    Module.Write Material Random=true
    Module.Write Position=true
    Module.Write Scale=true
    Module.Write SpriteRotation=false
    Module.Write SpriteSize=false
    Position=&Add_Vector(
     A=/Niagara/DynamicInputs/Helpers/SimulationPosition()
     B=&Multiply_VectorByFloat(Vector=$Engine.Owner.SystemXAxis, Float=-150)
    )
   )
   &SkeletalMeshLocation(
    Apply Sampled Orientation to Mesh Particles=true
    Bone / Socket Index=2
    Inherit Velocity=false
    "Inherit Velocity (Scale)"=-0.2
    Interpolate Mesh Sampling=false
    MaskSpawnGroup=false
    Mesh Sampling Type=::NewEnumerator5
    Sampled Position Offset=&Lerp_Vector(A=<-180 0 0>, B=<-25 0 5>, Alpha=&ClampFloat(Float=$User.ForwardDot, Max=0))
    Sampled Position Offset Coordinate Space:ENiagaraMeshTransforms=::NewEnumerator3
    Skeletal Mesh=$User.SkeletalMesh
    Socket Sampling Mode=::NewEnumerator5
   ) {module=SkeletalMeshLocation002}
   &AddVelocity(Velocity=<0 0 20>)
   /SetVariables(Particles.Scale=&Multiply_Vector(
     A=$Particles.Scale
     B=&Lerp_Vector(
      A=<0 0 0>
      B=<0.8 2 0>
      Alpha=&ClampFloat(Float=$System.Timeline.Time, Min=&FN_FrontEndFloatModifier(Gameplay=0, Front End=1))
     )
    )) {enabled=false, isSetParameters=true}
   &SkeletalMeshLocation(
    Apply Sampled Orientation to Mesh Particles=true
    Bone / Socket Index=$Particles.SpawnGroup
    Inherit Velocity=false
    Interpolate Mesh Sampling=false
    MaskSpawnGroup=false
    Mesh Sampling Type=::NewEnumerator5
    Sampled Position Offset=&Lerp_Vector(
     A=<50 0 0>
     B=&Lerp_Vector(A=<-75 -15 0>, B=<-75 15 1>, Alpha=&MakeFloatFromInt(Int=$Particles.SpawnGroup))
     Alpha=$Particles.FrontToBack
    )
    Sampled Position Offset Coordinate Space:ENiagaraMeshTransforms=::NewEnumerator3
    Skeletal Mesh=$User.SkeletalMesh
    Socket Sampling Mode=::NewEnumerator5
   ) {enabled=false}
   &InitialMeshOrientation(
    Orientation Vector=&Multiply_Vector(
     A=&Add_Vector(
      A=&Lerp_Vector(A=<0 0 0>, B=<-1 0 0>, Alpha=$Particles.FrontToBack)
      B=&Lerp_Vector(A=<-1 0.8 0>, B=<-0.75 1 0>, Alpha=$Particles.LeanLerp)
     )
     B=&Multiply_Vector(
      A=&Lerp_Vector(A=<1 1 1>, B=<1 -1 1>, Alpha=&MakeFloatFromInt(Int=$Particles.SpawnGroup))
      B=&Lerp_Vector(A=<1 1 1>, B=<1 0.75 1>, Alpha=&ClampFloat(Float=$User.ForwardDot))
     )
    )
    Rotation Coordinate Space=::NewEnumerator0
    Use Orientation Vector=true
    Use Rotation Vector=false
   )
   &InitialMeshOrientation(
    Orientation Vector=&Multiply_Vector(
     A=&Add_Vector(
      A=&VectorFromCurve(
       VectorCurve=DataInterface() {propertyValues {
         XCurve {keys [<at 0.33 1 RCIM_Cubic>, {arriveTangent=-1.96, interpMode=RCIM_Cubic, leaveTangent=-1.96, tangentMode=RCTM_User, time=1, value=0.2}]}
         YCurve {keys [<at 0.33 0.5 RCIM_Cubic>, {arriveTangent=-0.6339, interpMode=RCIM_Cubic, leaveTangent=-0.6339, tangentMode=RCTM_User, time=1, value=0.2}]}
         ZCurve {keys [<at 0 0>,<at 1 0>]}
        }}
       CurveIndex=$Particles.FrontToBack
       Scale Curve=<-1 1 1>
      )
      B=&Lerp_Vector(A=<0 0 0>, B=<0.5 0 0>, Alpha=$Particles.LeanLerp)
     )
     B=&Multiply_Vector(
      A=&Lerp_Vector(A=<1 1 1>, B=<1 -1 1>, Alpha=&MakeFloatFromInt(Int=$Particles.SpawnGroup))
      B=&Lerp_Vector(A=<1 1 1>, B=<1 0.4 1>, Alpha=$User.ForwardDot)
     )
    )
    Rotation Coordinate Space=::NewEnumerator0
    Use Orientation Vector=true
    Use Rotation Vector=false
   ) {enabled=false, module=InitialMeshOrientation001}
   /SetVariables(Particles.Scale=&Multiply_Vector(
     A=&VectorFromCurve(
      VectorCurve=DataInterface() {propertyValues {
        XCurve {keys [{arriveTangent=1.7846, interpMode=RCIM_Cubic, leaveTangent=1.7846, tangentMode=RCTM_User, time=0, value=0.4}, <at 1 1 RCIM_Cubic>]}
        YCurve {keys [{arriveTangent=1.7846, interpMode=RCIM_Cubic, leaveTangent=1.7846, tangentMode=RCTM_User, time=0, value=0.4}, <at 1 1 RCIM_Cubic>]}
        ZCurve {keys [{arriveTangent=1.7846, interpMode=RCIM_Cubic, leaveTangent=1.7846, tangentMode=RCTM_User, time=0, value=0.4}, <at 1 1 RCIM_Cubic>]}
       }}
      CurveIndex=$Particles.FrontToBack
      Scale Curve=&Multiply_VectorByFloat(
       Vector=$Particles.Scale
       Float=&Lerp_Float(
        A=&FN_FrontEndFloatModifier(Gameplay=0, Front End=0.4)
        B=0.6
        Alpha=&ClampFloat(Float=$System.Timeline.Time)
       )
      )
     )
     B=&Lerp_Vector(A=<1.5 1 0.2>, B=<1.5 1 1>, Alpha=$Particles.LeanLerp)
    )) {enabled=false, isSetParameters=true, module=SetVariables_#3}
   /SetVariables(Particles.Scale=&Multiply_Vector(
     A=&VectorFromFloat(Value=&FloatFromCurve(
       FloatCurve=DataInterface() {propertyValues {Curve {keys [{interpMode=RCIM_Cubic, tangentMode=RCTM_User, time=0, value=0.4}, {interpMode=RCIM_Cubic, tangentMode=RCTM_User, time=1, value=1}]}}}
       CurveIndex=$Particles.FrontToBack
       Scale Curve=&Multiply_Float(A=1, B=&Lerp_Float(A=1, B=1.4, Alpha=$User.ForwardDot))
      ))
     B=&Lerp_Vector(A=<0.6 0.6 0.6>, B=<0.8 0.8 0.8>, Alpha=$Particles.LeanLerp)
    )) {enabled=false, isSetParameters=true, module=SetVariables_#4}
  ]
  particleUpdateScript [
   &ParticleState()
   &SkeletalMeshLocation(
    Apply Sampled Orientation to Mesh Particles=true
    Bone / Socket Index=2
    Inherit Velocity=false
    MaskSpawnGroup=false
    Mesh Sampling Type=::NewEnumerator5
    Sampled Position Offset=&Add_Vector(
     A=&Add_Vector(
      A=<-80 0 10>
      B=&Lerp_Vector(A=<0 5 -8>, B=<0 -10 -8>, Alpha=&MakeFloatFromInt(Int=$Particles.SpawnGroup))
     )
     B=&VectorFromCurve(
      VectorCurve=DataInterface() {propertyValues {
        XCurve {keys [<at 0 0>, {arriveTangent=-382.5781, interpMode=RCIM_Cubic, leaveTangent=-382.5781, tangentMode=RCTM_User, time=1, value=-125}]}
        YCurve {keys [<at 0.33 0 RCIM_Cubic>, {arriveTangent=-39.3211, interpMode=RCIM_Cubic, leaveTangent=-39.3211, tangentMode=RCTM_User, time=1, value=10}]}
        ZCurve {keys [<at 0 0>,<at 1 10>]}
       }}
      CurveIndex=$Particles.FrontToBack
      Scale Curve=&Lerp_Vector(A=<0.2 1 1>, B=<0.5 0 1>, Alpha=&MakeFloatFromInt(Int=$Particles.SpawnGroup))
     )
    )
    Sampled Position Offset Coordinate Space:ENiagaraMeshTransforms=::NewEnumerator3
    Skeletal Mesh=$User.SkeletalMesh
    Socket Sampling Mode=::NewEnumerator5
   ) {enabled=false, module=SkeletalMeshLocation001}
   &ScaleColor(
    Scale Alpha=&FloatFromCurve(FloatCurve=DataInterface() {propertyValues {Curve {keys [
         {arriveTangent=9.9926, interpMode=RCIM_Cubic, leaveTangent=9.9927, tangentMode=RCTM_User, time=0, value=0}
         {arriveTangent=-2.8339, interpMode=RCIM_Cubic, leaveTangent=-2.8339, tangentMode=RCTM_User, time=0.0216, value=1}
         <at 1.0011 0.002 RCIM_Cubic>
        ]}}})
    ScaleA=true
    ScaleRGB=true
   )
   &Drag(Drag=0, Rotational Drag=20, Use Linear Drag=true, Use Rotational Drag=true)
   &SolveForcesAndVelocity(Clamp Velocity=false, Limit Acceleration=false)
   /SetVariables(Particles.Scale=&Multiply_Vector(
     A=$Particles.Initial.Scale
     B=&VectorFromCurve(VectorCurve=DataInterface() {propertyValues {
        XCurve {keys [
          {arriveTangent=4.6994, interpMode=RCIM_Cubic, leaveTangent=4.6994, tangentMode=RCTM_User, time=0, value=0.0795}
          {arriveTangent=0.4589, interpMode=RCIM_Cubic, leaveTangent=0.4589, tangentMode=RCTM_User, time=0.15, value=0.7609}
          <at 1 1>
         ]}
        YCurve {keys [
          {arriveTangent=1.7936, interpMode=RCIM_Cubic, leaveTangent=1.7936, tangentMode=RCTM_User, time=0, value=0}
          <at 0.4 0.4>
          {arriveTangent=-0.1215, interpMode=RCIM_Cubic, leaveTangent=-0.1215, tangentMode=RCTM_User, time=1, value=1}
         ]}
        ZCurve {keys [<at 0 0 RCIM_Cubic>,<at 0.33 1 RCIM_Cubic>,<at 0.9988 0>]}
       }})
    )) {isSetParameters=true}
   &DynamicMaterialParameters(
    Index 0 Param 1=&FloatFromCurve(FloatCurve=DataInterface() {propertyValues {Curve {keys [
         {arriveTangent=0.2615, interpMode=RCIM_Cubic, leaveTangent=0.2615, tangentMode=RCTM_User, time=0, value=-1.7984}
         {arriveTangent=3.2109, interpMode=RCIM_Cubic, leaveTangent=3.2109, tangentMode=RCTM_User, time=0.9985, value=0}
        ]}}})
    Index 0 Param 2=&Add_Float(
     A=&FloatFromCurve(FloatCurve=DataInterface() {propertyValues {Curve {keys [
          {arriveTangent=-9.4845, interpMode=RCIM_Cubic, leaveTangent=-9.4845, tangentMode=RCTM_User, time=0, value=4}
          {arriveTangent=0.2383, interpMode=RCIM_Cubic, leaveTangent=0.2383, tangentMode=RCTM_User, time=1, value=0}
         ]}}})
     B=0
    )
    Index 1 Param 1=&FloatFromCurve(FloatCurve=DataInterface() {propertyValues {Curve {keys [
         <at 0 0.1 RCIM_Cubic>
         {arriveTangent=0.9, interpMode=RCIM_Cubic, leaveTangent=0.9, time=0.5861, value=0.2}
         {arriveTangent=3.3294, interpMode=RCIM_Cubic, leaveTangent=3.3294, tangentMode=RCTM_User, time=1, value=1}
        ]}}})
    Module.Param0WriteEnabled=true
    Module.Param0WriteEnabled1=true
    Module.Param1WriteEnabled=true
    Module.Param1WriteEnabled1=true
    Module.Param2WriteEnabled=true
    Module.Param2WriteEnabled1=true
    Module.Param3WriteEnabled=true
    Module.Param3WriteEnabled1=true
    Write Parameter Index 1=true
   ) {module=DynamicMaterialParameters001}
  ]
  rendererClasses [&NiagaraMeshRendererProperties]
  renderers [
   {rendererClass=&NiagaraMeshRendererProperties, rendererIndex=0}
   {rendererClass=&NiagaraMeshRendererProperties, rendererIndex=1}
   {rendererClass=&NiagaraMeshRendererProperties, rendererIndex=2}
  ]
 }
 FN_Water_Foam_SnapToSurface001 {
  bEnabled=false
  emitterUpdateScript [
   &EmitterState() {module=EmitterState001}
   &EmitterState()
   &EmitterScalability(
    MaxDistance=2000
    Module.bApplySpawnCountScale=true
    Module.bApplySpawnCountScaleByDistanceFraction=true
    Module.bUseMaxDistance=true
    Module.bUseMinDistance=false
    SpawnCountScaleByDistanceFraction=DataInterface() {propertyValues {Curve {keys [<at 0 1>,<at 1 0.5>]}}}
   ) {enabled=false}
   &SpawnPerUnit(Module.Use Max Movement Threshold=true, Module.Use Movement Tolerance=false, Spawn Per Unit=0, Use Spawn Probability=false) {enabled=false}
   &SpawnBurst_Instantaneous(Spawn Count=1, Use Spawn Probability=false) {enabled=false}
   &EmitterLifeCycle(MaxLoopCount=1, NextLoopDuration=1) {enabled=false}
   &SpawnRate(SpawnRate=10, Use Spawn Probability=false)
   &SpawnPerUnit(Module.Use Max Movement Threshold=true, Module.Use Movement Tolerance=false, Use Spawn Probability=false) {enabled=false, module=SpawnPerUnit001}
  ]
  particleSpawnScript [
   &InitializeParticle(
    Lifetime=&UniformRangedFloat(Minimum=1, Maximum=2)
    Mass=&UniformRangedFloat(Minimum=0.6, Maximum=1.2)
    Mesh Scale=&VectorFromFloat(Value=300)
    Module.Write Color=true
    Module.Write Lifetime=true
    Module.Write Mass=true
    Module.Write Material Random=true
    Module.Write Position=false
    Module.Write Scale=true
    Module.Write SpriteRotation=true
    Module.Write SpriteSize=false
    Position=/Niagara/Functions/Localspace/SimulationPosition()
    Sprite Rotation=&NormalizedAngleToDegrees(NormalizedAngle=&UniformRangedFloat(Minimum=0, Maximum=1))
   ) {enabled=false}
   /Niagara/Modules/Spawn/Initialization/V2/InitializeParticle(
    Lifetime=&UniformRangedFloat(Minimum=2, Maximum=3)
    Mass=&UniformRangedFloat(Minimum=0.6, Maximum=1.2)
    Mass Mode=::NewEnumerator1
    Mesh Scale Mode=::NewEnumerator3
    Mesh Uniform Scale=100
    Sprite Rotation Angle=&UniformRangedFloat(Minimum=0, Maximum=1)
    Sprite Rotation Mode=::NewEnumerator1
    UsePositionOffset=false
    Write Material Random=true
   ) {module=InitializeParticle001}
   /Niagara/Modules/Spawn/Orientation/V2/InitialMeshOrientation(
    Orientation Vector=<0 0 0>
    Rotation=&UniformRangedVector(Minimum=<0 0 0>, Maximum=<0 0 1>)
    Use Orientation Vector=true
    Use Rotation Vector=true
   ) {enabled=false}
   /Niagara/Modules/Spawn/Orientation/V3/InitialMeshOrientation(
    Mesh Orientation Mode=::NewEnumerator2
    Orientation Vector=<0 0 0>
    Rotation=&UniformRangedVector(Minimum=<0 0 0>, Maximum=<0 0 1>)
    Use Rotation Vector=true
   ) {module=InitialMeshOrientation001}
   &DynamicMaterialParameters(
    Index 1 Param 3=400
    Module.Param0WriteEnabled1=false
    Module.Param1WriteEnabled1=false
    Module.Param2WriteEnabled1=true
    Module.Param3WriteEnabled1=false
    Write Parameter Index 0=false
    Write Parameter Index 1=true
   )
   &InitializeWaterSnapToOwner() {enabled=false}
   &InheritVelocity(Module.Limit Speed=true, Velocity Scale=<100 1 1>) {enabled=false}
  ]
  particleUpdateScript [
   &ParticleState()
   &SkeletalMeshLocation(Bone Sampling Mode=::NewEnumerator4, MaskSpawnGroup=false, Sampled Position Offset=<0 0 -10>, Skeletal Mesh=$User.SkeletalMesh) {enabled=false}
   &DynamicMaterialParameters(
    Index 0 Param 1=&FloatFromCurve(FloatCurve=DataInterface() {propertyValues {Curve {keys [<at -0.0046 1>,<at 1.0011 4>]}}})
    Index 0 Param 2=&FloatFromCurve(FloatCurve=DataInterface() {propertyValues {Curve {keys [{arriveTangent=0.6218, interpMode=RCIM_Cubic, leaveTangent=0.6218, tangentMode=RCTM_User, time=-0.0046, value=0.5}, <at 1.0011 1 RCIM_Cubic>]}}})
    Index 0 Param 3=&FloatFromCurve(FloatCurve=DataInterface() {propertyValues {Curve {keys [<at -0.0046 1>,<at 1.0011 0.5>]}}})
    Index 2 Param 1=&FloatFromCurve(FloatCurve=DataInterface() {propertyValues {Curve {keys [<at -0.0046 0.4>,<at 1.0011 1.5>]}}})
    Index 2 Param 2=&FloatFromCurve(FloatCurve=DataInterface() {propertyValues {Curve {keys [<at -0.0046 0.1>,<at 0.11 2.6>,<at 1.0011 0.3>]}}})
    Index 2 Param 3=0.5
    Index 2 Param 4=&FloatFromCurve(FloatCurve=DataInterface() {propertyValues {Curve {keys [
         {arriveTangent=-1.9991, interpMode=RCIM_Cubic, time=0, value=1}
         {arriveTangent=-2.7143, interpMode=RCIM_Cubic, leaveTangent=-2.7143, tangentMode=RCTM_User, time=0.2095, value=0.6901}
         <at 1 0 RCIM_Cubic>
        ]}}})
    Module.Param0WriteEnabled=true
    Module.Param0WriteEnabled2=true
    Module.Param1WriteEnabled=true
    Module.Param1WriteEnabled2=true
    Module.Param2WriteEnabled=true
    Module.Param2WriteEnabled2=true
    Module.Param3WriteEnabled=true
    Module.Param3WriteEnabled2=true
    Write Parameter Index 2=true
   ) {module=DynamicMaterialParameters001}
   &ScaleColor(
    Scale Alpha=&FloatFromCurve(FloatCurve=DataInterface() {propertyValues {Curve {keys [<at -0.0046 1.0113 RCIM_Cubic>,<at 1.0011 0.002 RCIM_Cubic>]}}})
    ScaleA=true
    ScaleRGB=true
   ) {enabled=false}
   &Drag(Drag=2, Use Linear Drag=true, Use Rotational Drag=false)
   &SolveForcesAndVelocity(Clamp Velocity=false, Limit Acceleration=false)
   &ScaleMeshSize(Scale Factor=&VectorFromFloat(Value=&FloatFromCurve(FloatCurve=DataInterface() {propertyValues {Curve {keys [
          {arriveTangent=26.4511, interpMode=RCIM_Cubic, leaveTangent=26.451, tangentMode=RCTM_User, time=-0.0046, value=0.6073}
          {arriveTangent=5.8301, interpMode=RCIM_Cubic, leaveTangent=5.8301, time=0.1863, value=3.3239}
          <at 1.0011 6.471 RCIM_Cubic>
         ]}}}))) {module=MeshSizeScale}
   &ConstrainPositionToPlane(Plane Normal:Vector3f=$Particles.WaterPlaneNormal, Plane Position:Vector3f=$Particles.WaterPlaneLocation)
  ]
  rendererClasses [&NiagaraMeshRendererProperties]
  renderers [{rendererClass=&NiagaraMeshRendererProperties, rendererIndex=0}]
 }
 FN_Water_Foam_SnapToSurface_Detail {
  bEnabled=false
  emitterUpdateScript [
   &EmitterState() {module=EmitterState001}
   &EmitterState()
   &EmitterScalability(
    MaxDistance=2000
    Module.bApplySpawnCountScale=true
    Module.bApplySpawnCountScaleByDistanceFraction=true
    Module.bUseMaxDistance=true
    Module.bUseMinDistance=false
    SpawnCountScaleByDistanceFraction=DataInterface() {propertyValues {Curve {keys [<at 0 1>,<at 1 0.5>]}}}
   ) {enabled=false}
   &SpawnPerUnit(
    Module.Use Max Movement Threshold=true
    Module.Use Movement Tolerance=true
    Movement Tolerance=5
    Spawn Per Unit=200
    Use Spawn Probability=false
   ) {enabled=false}
   &EmitterLifeCycle(MaxLoopCount=1, NextLoopDuration=1) {enabled=false}
   &SpawnRate(SpawnRate=20, Use Spawn Probability=false)
  ]
  particleSpawnScript [
   /Niagara/Modules/Spawn/Initialization/InitializeParticle(
    Lifetime=&UniformRangedFloat(Minimum=3, Maximum=4)
    Mass=&UniformRangedFloat(Minimum=0.6, Maximum=1.2)
    Mesh Scale=&VectorFromFloat(Value=&UniformRangedFloat(Minimum=100, Maximum=200))
    Module.Write Color=true
    Module.Write Lifetime=true
    Module.Write Mass=true
    Module.Write Material Random=true
    Module.Write Position=false
    Module.Write Scale=true
    Module.Write SpriteRotation=true
    Module.Write SpriteSize=false
    Position=&SimulationPosition()
    Sprite Rotation=&NormalizedAngleToDegrees(NormalizedAngle=&UniformRangedFloat(Minimum=0, Maximum=1))
    Sprite Size=<2 2>
   ) {enabled=false}
   /Niagara/Modules/Spawn/Initialization/V2/InitializeParticle(
    Lifetime=&UniformRangedFloat(Minimum=2, Maximum=3)
    Mass=&UniformRangedFloat(Minimum=0.6, Maximum=1.2)
    Mass Mode=::NewEnumerator1
    Mesh Scale Mode=::NewEnumerator3
    Mesh Uniform Scale=100
    Sprite Rotation Angle=&UniformRangedFloat(Minimum=0, Maximum=1)
    Sprite Rotation Mode=::NewEnumerator1
    UsePositionOffset=false
    Write Material Random=true
   ) {module=InitializeParticle001}
   /Niagara/Modules/Spawn/Orientation/V2/InitialMeshOrientation(
    Orientation Vector=<0 0 0>
    Rotation=&UniformRangedVector(Minimum=<0 0 0>, Maximum=<0 0 1>)
    Use Orientation Vector=true
    Use Rotation Vector=true
   ) {enabled=false}
   &CylinderLocation(
    Cylinder Height=0
    Cylinder Height Midpoint=0
    Cylinder Radius=150
    MaskSpawnGroup=false
    Non Uniform Scale:Vector3f=<1 1 0>
    Override Seed=false
    UseOwnerScale=true
   )
   /Niagara/Modules/Spawn/Orientation/V3/InitialMeshOrientation(
    Mesh Orientation Mode=::NewEnumerator2
    Orientation Vector=<0 0 0>
    Rotation=&UniformRangedVector(Minimum=<0 0 0>, Maximum=<0 0 1>)
    Use Rotation Vector=true
   ) {module=InitialMeshOrientation001}
   &DynamicMaterialParameters(
    Index 1 Param 3=200
    Module.Param0WriteEnabled1=false
    Module.Param1WriteEnabled1=false
    Module.Param2WriteEnabled1=true
    Module.Param3WriteEnabled1=false
    Write Parameter Index 0=false
    Write Parameter Index 1=true
   )
   &InitializeWaterSnapToOwner() {enabled=false}
   &InheritVelocity(Module.Limit Speed=true, Velocity Scale=<0.8 0.8 0.8>) {enabled=false}
  ]
  particleUpdateScript [
   &ParticleState()
   /Game/Effects/Niagara/Emitters/Water/FN_DistanceKillParticles(AgeReduiceAmount=0.01) {enabled=false}
   &DynamicMaterialParameters(
    Index 0 Param 1=&FloatFromCurve(FloatCurve=DataInterface() {propertyValues {Curve {keys [<at -0.0046 1>,<at 1.0011 4>]}}})
    Index 0 Param 2=&FloatFromCurve(FloatCurve=DataInterface() {propertyValues {Curve {keys [<at -0.0046 0.2>,<at 1.0011 1.3>]}}})
    Index 0 Param 3=&FloatFromCurve(FloatCurve=DataInterface() {propertyValues {Curve {keys [<at -0.0046 1>,<at 1.0011 0.5>]}}})
    Index 2 Param 1=&FloatFromCurve(FloatCurve=DataInterface() {propertyValues {Curve {keys [<at -0.0046 0.4>,<at 1.0011 1.5>]}}})
    Index 2 Param 2=&FloatFromCurve(FloatCurve=DataInterface() {propertyValues {Curve {keys [<at -0.0046 0.1>,<at 0.2854 2.6>,<at 1.0011 0.3>]}}})
    Index 2 Param 3=0.8
    Index 2 Param 4=&FloatFromCurve(FloatCurve=DataInterface() {propertyValues {Curve {keys [
         {arriveTangent=-1.9991, interpMode=RCIM_Cubic, time=0, value=1}
         {arriveTangent=-2.7143, interpMode=RCIM_Cubic, leaveTangent=-2.7143, tangentMode=RCTM_User, time=0.2095, value=0.6901}
         <at 1 0 RCIM_Cubic>
        ]}}})
    Module.Param0WriteEnabled=true
    Module.Param0WriteEnabled2=true
    Module.Param1WriteEnabled=true
    Module.Param1WriteEnabled2=true
    Module.Param2WriteEnabled=true
    Module.Param2WriteEnabled2=true
    Module.Param3WriteEnabled=true
    Module.Param3WriteEnabled2=true
    Write Parameter Index 2=true
   ) {module=DynamicMaterialParameters001}
   &ScaleColor(
    Scale Alpha=&FloatFromCurve(FloatCurve=DataInterface() {propertyValues {Curve {keys [<at -0.0046 1.0113 RCIM_Cubic>,<at 1.0011 0.002 RCIM_Cubic>]}}})
    ScaleA=true
    ScaleRGB=true
   ) {enabled=false}
   &Drag(Drag=2, Use Linear Drag=true, Use Rotational Drag=false)
   &SolveForcesAndVelocity(Clamp Velocity=false, Limit Acceleration=false)
   &ScaleMeshSize(Scale Factor=&VectorFromFloat(Value=1)) {module=MeshSizeScale}
   &ConstrainPositionToPlane(Plane Normal:Vector3f=$Particles.WaterPlaneNormal, Plane Position:Vector3f=$Particles.WaterPlaneLocation) {enabled=false}
  ]
  rendererClasses [&NiagaraMeshRendererProperties]
  renderers [{rendererClass=&NiagaraMeshRendererProperties, rendererIndex=0}]
 }
 Smokes001 {
  bEnabled=false
  emitterUpdateScript [
   &EmitterState()
   &EmitterScalability(
    Module.bApplySpawnCountScale=false
    Module.bApplySpawnCountScaleByDistanceFraction=false
    Module.bUseMaxDistance=false
    Module.bUseMinDistance=false
   ) {enabled=false}
   &EmitterLifeCycle() {enabled=false}
   &SpawnBurst_Instantaneous(Spawn Count=5, Use Spawn Probability=false) {enabled=false}
   &SpawnRate(SpawnRate=&Lerp_Float(A=0, B=10, Alpha=$System.Timeline.Time), Use Spawn Probability=false)
  ]
  particleSpawnScript [
   &FN_SetSpawnGroup(Number of Spawn Groups=2)
   /SetVariables(
    Particles.FrontToBack:NiagaraFloat=&RandomRangeFloat(Minimum=0, Maximum=0.5)
    Particles.LeanLerp:NiagaraFloat=&Multiply_Float(
     A=$User.RightDot
     B=&Lerp_Float(A=-1, B=1, Alpha=&MakeFloatFromInt(Int=$Particles.SpawnGroup))
    )
   ) {isSetParameters=true}
   /SetVariables(Particles.Random=&UniformRangedFloat(Minimum=0, Maximum=1)) {isSetParameters=true}
   /Niagara/Modules/Spawn/Initialization/InitializeParticle(
    Color=<rgba 1.5 1.5 1.5 1>
    Lifetime=0.4
    Module.Write Color=true
    Module.Write Lifetime=true
    Module.Write Mass=true
    Module.Write Material Random=true
    Module.Write Position=false
    Module.Write Scale=false
    Module.Write SpriteRotation=true
    Module.Write SpriteSize=true
    Position=&SimulationPosition()
    Sprite Rotation=&NormalizedAngleToDegrees(NormalizedAngle=&UniformRangedFloat(Minimum=0, Maximum=1))
    Sprite Size=&Vector2DFromFloat(Value=&UniformRangedFloat(Minimum=400, Maximum=600))
   ) {module=InitializeParticle001}
   &InitializeParticle(
    Lifetime=&UniformRangedFloat(Minimum=0, Maximum=1)
    Module.Write Color=true
    Module.Write Lifetime=true
    Module.Write Mass=true
    Module.Write Material Random=true
    Module.Write Position=false
    Module.Write Scale=false
    Module.Write SpriteRotation=true
    Module.Write SpriteSize=true
    Position=&SimulationPosition()
    Sprite Rotation=&UniformRangedFloat(Minimum=0, Maximum=360)
    Sprite Size=&Vector2DFromFloat(Value=&UniformRangedFloat(Minimum=10, Maximum=20))
   ) {enabled=false}
   &DynamicMaterialParameters(
    Index 0 Param 1=&UniformRangedFloat(Minimum=1, Maximum=2)
    Index 0 Param 2=&UniformRangedFloat(Minimum=1, Maximum=3)
    Index 0 Param 3=1
    Module.Param0WriteEnabled=true
    Module.Param1WriteEnabled=true
    Module.Param2WriteEnabled=true
    Module.Param3WriteEnabled=false
   )
   &SkeletalMeshLocation(
    Apply Sampled Orientation to Mesh Particles=true
    Bone / Socket Index=$Particles.SpawnGroup
    Inherit Velocity=false
    Interpolate Mesh Sampling=false
    MaskSpawnGroup=false
    Mesh Sampling Type=::NewEnumerator5
    Sampled Position Offset=&Lerp_Vector(A=<0 40 0>, B=<0 -40 0>, Alpha=&MakeFloatFromInt(Int=$Particles.SpawnGroup))
    Sampled Position Offset Coordinate Space:ENiagaraMeshTransforms=::NewEnumerator3
    Skeletal Mesh=$User.SkeletalMesh
    Socket Sampling Mode=::NewEnumerator5
   ) {module=SkeletalMeshLocation001}
   &AddVelocityInCone(
    Cone Angle=0
    Cone Axis=&Lerp_Vector(A=<0.2 1 0>, B=<-0.2 1 0>, Alpha=&MakeFloatFromInt(Int=$Particles.SpawnGroup))
    Module.Use Velocity Falloff On Cone Axis=false
    Velocity Strength=&Lerp_Float(A=80, B=400, Alpha=$System.Timeline.Time)
   )
   &SkeletalMeshLocation(MaskSpawnGroup=false, Mesh Sampling Type=::NewEnumerator5, Skeletal Mesh=$User.SkeletalMesh, Socket Sampling Mode=::NewEnumerator5) {enabled=false}
  ]
  particleUpdateScript [
   &ParticleState()
   &ScaleSpriteSize(Scale Factor=&Vector2DFromFloat(Value=&FloatFromCurve(FloatCurve=DataInterface() {propertyValues {Curve {keys [<at -0.0046 1.0113>,<at 1.0011 2>]}}}, Scale Curve=0.15)))
   /Niagara/Modules/Update/SubUV/SubUVAnimation(Number Of Frames=&FloatFromCurve(FloatCurve=DataInterface() {propertyValues {Curve {keys [<at -0.0046 64>,<at 1.0011 0.002>]}}}))
   &ScaleColor(
    Scale Alpha=&FloatFromCurve(FloatCurve=DataInterface() {propertyValues {Curve {keys [<at 0 0 RCIM_Cubic>, {arriveTangent=0.002, interpMode=RCIM_Cubic, leaveTangent=0.002, time=0.1942, value=2}, <at 1.0011 0.002 RCIM_Cubic>]}}})
    ScaleA=true
    ScaleRGB=true
   )
   &Drag(Use Linear Drag=true, Use Rotational Drag=false)
   &InheritSourceMovement(ApplyForce=false, ApplyPosition=true, ApplyPositionFalloff=false) {enabled=false}
   &SolveForcesAndVelocity(Clamp Velocity=false, Limit Acceleration=false)
   &CameraOffset(Camera Offset Amount=100)
   &DynamicMaterialParameters(
    Index 0 Param 1=&UniformRangedFloat(Minimum=1, Maximum=2)
    Index 0 Param 3=2
    Index 0 Param 4=0.01
    Module.Param0WriteEnabled=false
    Module.Param1WriteEnabled=true
    Module.Param2WriteEnabled=true
    Module.Param3WriteEnabled=true
   ) {module=DynamicMaterialParameters001}
  ]
  rendererClasses [&NiagaraSpriteRendererProperties]
  renderers [{rendererClass=&NiagaraSpriteRendererProperties}]
 }
 Speedlines {
  bEnabled=false
  emitterUpdateScript [
   &EmitterState(Enable Distance Culling=true, MaxDistance=2048, Module.bUseMaxDistance=true, Module.bUseMinDistance=false, Scalability Mode=::NewEnumerator1)
   &SpawnRate(
    SpawnRate=&Multiply_Float(
     A=&FN_FrontEndFloatModifier(Gameplay=1, Front End=1)
     B=&UniformRangedFloat(Minimum=75, Maximum=150)
    )
    Use Spawn Probability=false
   )
  ]
  particleSpawnScript [
   &InitializeParticle(
    Color=&MakeLinearColorFromVectorAndFloat(
     "Vector (RGB)"=<1 1 1>
     "Float (Alpha)"=&Multiply_Float(
      A=&Multiply_Float(
       A=&UniformRangedFloat(Minimum=0.015, Maximum=0.15)
       B=&FN_FrontEndFloatModifier(Gameplay=2, Front End=0.5)
      )
      B=&Add_Float(
       A=1
       B=&Multiply_Float(A=10, B=&Subtract_Float(A=1, B=$Engine.Emitter.SpawnCountScale))
      )
     )
    )
    Lifetime=&FN_FrontEndFloatModifier(
     Gameplay=&UniformRangedFloat(Minimum=0.1, Maximum=0.2)
     Front End=&UniformRangedFloat(Minimum=0.1, Maximum=0.2)
    )
    Module.Write Color=true
    Module.Write Lifetime=true
    Module.Write Mass=true
    Module.Write Material Random=true
    Module.Write Position=false
    Module.Write Scale=false
    Module.Write SpriteRotation=false
    Module.Write SpriteSize=true
    Position=&SimulationPosition()
    Sprite Size=&UniformRangedVector2D(Minimum=<3 4>, Maximum=<5 6>)
   )
   &SkeletalMeshLocation(
    Bone / Socket Index=3
    MaskSpawnGroup=false
    Mesh Sampling Type=::NewEnumerator5
    Sampled Position Offset=<60 0 0>
    Skeletal Mesh=$User.SkeletalMesh
    Socket Sampling Mode=::NewEnumerator5
   )
   &CylinderLocation(
    Cylinder Height=200
    Cylinder Radius=&UniformRangedFloat(Minimum=100, Maximum=200)
    MaskSpawnGroup=false
    Non Uniform Scale:Vector3f=<0.8 0 1>
    Offset=<0 -20 20>
    Offset Mode:ENiagara_OffsetMode=::NewEnumerator0
    Override Seed=false
    UseOwnerScale=true
   )
   &AddVelocity(Velocity=&FN_FrontEndVector3Modifier(Gameplay=<0 2000 0>, Front End=<0 3000 0>))
   &AddVelocityFromPoint(
    Module.Use Falloff Distance=false
    Velocity Strength=&Multiply_Float(
     A=&FN_FrontEndFloatModifier(Gameplay=0, Front End=2)
     B=&UniformRangedFloat(Minimum=500, Maximum=700)
    )
   )
  ]
  particleUpdateScript [
   &ParticleState()
   &SolveForcesAndVelocity(Clamp Velocity=false, Limit Acceleration=false)
   &ScaleSpriteSizeBySpeed(Max Scale Factor=<1 40>, Min Scale Factor=<0.01 0.05>, Velocity Threshold=1)
   &ScaleSpriteSize(Scale Factor=&Vector2DFromCurve(
     Vector2Curve=DataInterface() {propertyValues {bExposeCurve=true, XCurve {keys [<at 0 0.5 RCIM_Cubic>,<at 0.5 1 RCIM_Cubic>]}, YCurve {keys [<at 0 0 RCIM_Cubic>,<at 0.5 1 RCIM_Cubic>]}}}
     ScaleCurve=<0.5 0.5>
    ))
   &ScaleColor(
    Scale Alpha=&FloatFromCurve(
     FloatCurve=DataInterface() {propertyValues {Curve {keys [<at 0 0 RCIM_Cubic>, <at 0.2 1>, {arriveTangent=-2.221, interpMode=RCIM_Cubic, leaveTangent=-2.221, tangentMode=RCTM_User, time=1, value=0}]}}}
     Scale Curve=&FN_FrontEndFloatModifier(Gameplay=1, Front End=0)
    )
    ScaleA=true
    ScaleRGB=true
   )
  ]
  rendererClasses [&NiagaraSpriteRendererProperties]
  renderers [{rendererClass=&NiagaraSpriteRendererProperties}]
 }
 Speedlines001 {
  emitterUpdateScript [
   &EmitterState(Enable Distance Culling=true, MaxDistance=2048, Module.bUseMaxDistance=true, Module.bUseMinDistance=false, Scalability Mode=::NewEnumerator1)
   &SpawnRate(
    SpawnRate=&Multiply_Float(
     A=&FN_FrontEndFloatModifier(Gameplay=$User.ForwardDot, Front End=1)
     B=&Multiply_Float(A=&UniformRangedFloat(Minimum=75, Maximum=150), B=1)
    )
    Use Spawn Probability=false
   )
  ]
  particleSpawnScript [
   &InitializeParticle(
    Color=&MakeLinearColorFromVectorAndFloat(
     "Vector (RGB)"=<1 1 1>
     "Float (Alpha)"=&Multiply_Float(
      A=&Multiply_Float(
       A=&UniformRangedFloat(Minimum=0.015, Maximum=0.15)
       B=&FN_FrontEndFloatModifier(Gameplay=1, Front End=0.5)
      )
      B=&Add_Float(
       A=1
       B=&Multiply_Float(A=10, B=&Subtract_Float(A=1, B=$Engine.Emitter.SpawnCountScale))
      )
     )
    )
    Lifetime=&FN_FrontEndFloatModifier(
     Gameplay=&UniformRangedFloat(Minimum=0.2, Maximum=0.35)
     Front End=&UniformRangedFloat(Minimum=0.1, Maximum=0.15)
    )
    Module.Write Color=true
    Module.Write Lifetime=true
    Module.Write Mass=true
    Module.Write Material Random=true
    Module.Write Position=false
    Module.Write Scale=false
    Module.Write SpriteRotation=false
    Module.Write SpriteSize=true
    Position=&SimulationPosition()
    Sprite Size=&UniformRangedVector2D(Minimum=<3 4>, Maximum=<5 6>)
   )
   &SkeletalMeshLocation(
    Bone / Socket Index=3
    MaskSpawnGroup=false
    Mesh Sampling Type=::NewEnumerator5
    Sampled Position Offset=<100 0 60>
    Skeletal Mesh=$User.SkeletalMesh
    Socket Sampling Mode=::NewEnumerator5
   )
   &CylinderLocation(
    Cylinder Height=150
    Cylinder Radius=&UniformRangedFloat(Minimum=100, Maximum=150)
    MaskSpawnGroup=false
    Override Seed=false
    UseOwnerScale=true
   )
   &AddVelocity(Velocity=&FN_FrontEndVector3Modifier(Gameplay=<0 2000 0>, Front End=<0 3000 0>))
   &AddVelocityFromPoint(
    Module.Use Falloff Distance=false
    Velocity Strength=&Multiply_Float(
     A=&FN_FrontEndFloatModifier(Gameplay=1, Front End=2)
     B=&UniformRangedFloat(Minimum=500, Maximum=700)
    )
   ) {enabled=false}
  ]
  particleUpdateScript [
   &ParticleState()
   &SolveForcesAndVelocity(Clamp Velocity=false, Limit Acceleration=false)
   &ScaleSpriteSize(Scale Factor=&Vector2DFromCurve(
     Vector2Curve=DataInterface() {propertyValues {XCurve {keys [<at 0 0.5 RCIM_Cubic>,<at 0.5 1 RCIM_Cubic>]}, YCurve {keys [<at 0 0 RCIM_Cubic>,<at 0.5 1 RCIM_Cubic>]}}}
     ScaleCurve=&Vector2DFromFloat(Value=1)
    ))
   &ScaleSpriteSizeBySpeed(Max Scale Factor=<1 40>, Min Scale Factor=<0.01 0.05>, Velocity Threshold=1)
   &ScaleColor(
    Scale Alpha=&FloatFromCurve(
     FloatCurve=DataInterface() {propertyValues {Curve {keys [<at 0 0 RCIM_Cubic>, <at 0.2 1>, {arriveTangent=-2.221, interpMode=RCIM_Cubic, leaveTangent=-2.221, tangentMode=RCTM_User, time=1, value=0}]}}}
     Scale Curve=&FN_FrontEndFloatModifier(Gameplay=0.4, Front End=0)
    )
    ScaleA=true
    ScaleRGB=true
   )
  ]
  rendererClasses [&NiagaraSpriteRendererProperties]
  renderers [{rendererClass=&NiagaraSpriteRendererProperties}]
 }
 Splash_Drops {
  emitterUpdateScript [
   &EmitterState(Inactive Response=::NewEnumerator1, Life Cycle Mode=::NewEnumerator1, Loop Duration=1, UseLoopDelay=false)
   &SpawnRate(
    SpawnGroup:NiagaraInt32=&RandomRangeInt(Minimum=0, Maximum=1)
    SpawnRate=&Lerp_Float(
     A=&FN_FrontEndFloatModifier(Gameplay=0, Front End=30)
     B=30
     Alpha=&FloatFromCurve(
      FloatCurve=DataInterface() {propertyValues {Curve {keys [{interpMode=RCIM_Cubic, tangentMode=RCTM_User, time=0.5, value=0}, {interpMode=RCIM_Cubic, tangentMode=RCTM_User, time=1, value=1}]}}}
      CurveIndex=$System.Timeline001.Time
     )
    )
    Use Spawn Probability=false
   )
  ]
  particleSpawnScript [
   &FN_SetSpawnGroup(Number of Spawn Groups=2)
   /SetVariables(
    Particles.FrontToBack:NiagaraFloat=&RandomRangeFloat(Minimum=0.2, Maximum=0.4)
    Particles.LeanLerp:NiagaraFloat=&Multiply_Float(
     A=$User.RightDot
     B=&Lerp_Float(A=-1, B=1, Alpha=&MakeFloatFromInt(Int=$Particles.SpawnGroup))
    )
   ) {isSetParameters=true}
   /Niagara/Modules/Spawn/Initialization/V2/InitializeParticle(
    Color=<rgba 1 1 1 0.2>
    Lifetime Max=0.6
    Lifetime Min=0.3
    Lifetime Mode=::NewEnumerator1
    Sprite Rotation Mode=::NewEnumerator2
    Sprite Size Mode=::NewEnumerator1
    Uniform Sprite Size Max=6
    Uniform Sprite Size Min=4
    UsePositionOffset=false
    Write Material Random=true
   ) {module=InitializeParticle001}
   &SkeletalMeshLocation(
    Apply Sampled Orientation to Mesh Particles=true
    Bone / Socket Index=$Particles.SpawnGroup
    Inherit Velocity=false
    MaskSpawnGroup=false
    Mesh Sampling Type=::NewEnumerator5
    Sampled Position Offset=&VectorFromCurve(
     VectorCurve=DataInterface() {propertyValues {
       XCurve {keys [<at 0 0>, {arriveTangent=-382.5781, interpMode=RCIM_Cubic, leaveTangent=-382.5781, tangentMode=RCTM_User, time=1, value=-300}]}
       YCurve {keys [<at 0.33 4 RCIM_Cubic>, {arriveTangent=-92.7914, interpMode=RCIM_Cubic, leaveTangent=-92.7914, tangentMode=RCTM_User, time=1, value=0}]}
       ZCurve {keys [<at 0 5>,<at 1 10>]}
      }}
     CurveIndex=$Particles.FrontToBack
     Scale Curve=&Add_Vector(
      A=<0 0 0>
      B=&Lerp_Vector(A=<1.2 10 1>, B=<1.2 -10 1>, Alpha=&MakeFloatFromInt(Int=$Particles.SpawnGroup))
     )
    )
    Sampled Position Offset Coordinate Space:ENiagaraMeshTransforms=::NewEnumerator3
    Skeletal Mesh=$User.SkeletalMesh
    Socket Sampling Mode=::NewEnumerator5
   )
   /Niagara/Modules/Spawn/Location/SphereLocation(MaskSpawnGroup=false, Module.Override Randomness=false, Offset=<0 80 0>, Sphere Radius=3, Surface Only=false)
   &SubUVAnimation(SubUV Animation Mode=::NewEnumerator1)
   /SetVariables(Particles.Velocity=&Multiply_VectorByFloat(
     Vector=&ReturnMeshOrientationAxis(Mesh Orientation Axis=::XAxis, Mesh Orientation Quat=$Particles.MeshOrientation)
     Float=&Lerp_Float(A=-50, B=-50, Alpha=$User.ForwardDot)
    )) {isSetParameters=true}
   &AddVelocity(
    Scale Added Velocity=&Lerp_Vector(
     A=<1 1 1>
     B=<1.1 1.1 1.25>
     Alpha=&ClampFloat(
      Float=&Add_Float(A=&ClampFloat(Float=$User.ForwardDot), B=$Particles.LeanLerp)
      Min=-0.25
     )
    )
    Velocity=&Multiply_Vector(A=<1 5 1>, B=&RandomRangeVector(Minimum=<-40 40 300>, Maximum=<40 40 200>))
   )
   &AddVelocityFromPoint(Module.Use Falloff Distance=false, Origin Offset=<0 -20 -15>, Origin Offset Coordinate Space=::Simulation, Velocity Strength=150)
  ]
  particleUpdateScript [
   &ParticleState()
   &InheritVelocity(
    Module.Limit Speed=true
    Velocity Limit=10000
    Velocity Scale=&Multiply_VectorByFloat(Vector=<-0.03 -0.03 -0.1>, Float=0.1)
    Velocity Source=&TransformVector(Vector=$Engine.Owner.Velocity, Source Space=::World, Destination Space=::Local)
   ) {enabled=false}
   &SpriteRotationRate(Rotation Rate=&RandomRangeFloat(Minimum=-180, Maximum=180))
   &Drag(Drag=0.5, Use Linear Drag=true, Use Rotational Drag=false)
   &GravityForce(Gravity=<0 0 -900>)
   &InheritSourceMovement(ApplyForce=false, ApplyPosition=true, ApplyPositionFalloff=false) {enabled=false}
   &SolveForcesAndVelocity(Clamp Velocity=false, Limit Acceleration=false)
   &ScaleSpriteSize(Scale Factor=&Vector2DFromFloat(Value=&FloatFromCurve(
      FloatCurve=DataInterface() {propertyValues {Curve {keys [
          {interpMode=RCIM_Cubic, tangentMode=RCTM_User, time=0.8, value=1}
          {arriveTangent=-3.3474, interpMode=RCIM_Cubic, leaveTangent=-3.3474, tangentMode=RCTM_User, time=1, value=0}
         ]}}}
      Scale Curve=0.65
     )))
   &DynamicMaterialParameters(
    Index 0 Param 1=&RandomRangeFloat(Minimum=0, Maximum=1)
    Index 0 Param 3=&FloatFromCurve(FloatCurve=DataInterface() {propertyValues {Curve {keys [{interpMode=RCIM_Cubic, tangentMode=RCTM_User, time=0, value=0}, {interpMode=RCIM_Cubic, tangentMode=RCTM_User, time=1, value=1}]}}})
    Module.Param0WriteEnabled=true
    Module.Param1WriteEnabled=false
    Module.Param2WriteEnabled=true
    Module.Param3WriteEnabled=true
   )
  ]
  rendererClasses [&NiagaraSpriteRendererProperties]
  renderers [{rendererClass=&NiagaraSpriteRendererProperties}]
 }
 Splash_Foam {
  emitterUpdateScript [
   &EmitterState(Inactive Response=::NewEnumerator1, Life Cycle Mode=::NewEnumerator1, Loop Behavior=::NewEnumerator1, Loop Duration=3600, UseLoopDelay=false)
   &SpawnRate(
    SpawnGroup:NiagaraInt32=&RandomRangeInt(Minimum=0, Maximum=1)
    SpawnRate=&Lerp_Float(
     A=&FN_FrontEndFloatModifier(Gameplay=0, Front End=5)
     B=40
     Alpha=&FloatFromCurve(
      FloatCurve=DataInterface() {propertyValues {Curve {keys [{interpMode=RCIM_Cubic, tangentMode=RCTM_User, time=0.5, value=0}, {interpMode=RCIM_Cubic, tangentMode=RCTM_User, time=1, value=1}]}}}
      CurveIndex=$System.Timeline001.Time
     )
    )
    Use Spawn Probability=false
   )
  ]
  particleSpawnScript [
   &FN_SetSpawnGroup(Number of Spawn Groups=2)
   /SetVariables(
    Particles.FrontToBack:NiagaraFloat=&RandomRangeFloat(Minimum=0.2, Maximum=0.4)
    Particles.LeanLerp:NiagaraFloat=&Multiply_Float(
     A=$User.RightDot
     B=&Lerp_Float(A=-1, B=1, Alpha=&MakeFloatFromInt(Int=$Particles.SpawnGroup))
    )
   ) {isSetParameters=true}
   &InitializeParticle(
    Color=<rgba 40 40 40 0.1>
    Lifetime Max=0.4
    Lifetime Min=0.2
    Lifetime Mode=::NewEnumerator1
    Mesh Scale Mode=::NewEnumerator3
    Sprite Rotation Mode=::NewEnumerator2
    Sprite Size Mode=::NewEnumerator1
    Uniform Sprite Size Max=100
    Uniform Sprite Size Min=75
    UsePositionOffset=false
    Write Material Random=true
   ) {module=InitializeParticle001}
   &SkeletalMeshLocation(
    Apply Sampled Orientation to Mesh Particles=true
    Bone / Socket Index=$Particles.SpawnGroup
    Inherit Velocity=false
    Interpolate Mesh Sampling=false
    MaskSpawnGroup=false
    Mesh Sampling Type=::NewEnumerator5
    Sampled Position Offset=&Add_Vector(
     A=<-50 0 0>
     B=&VectorFromCurve(
      VectorCurve=DataInterface() {propertyValues {
        XCurve {keys [<at 0 0>, {arriveTangent=-382.5781, interpMode=RCIM_Cubic, leaveTangent=-382.5781, tangentMode=RCTM_User, time=1, value=-300}]}
        YCurve {keys [
          {arriveTangent=33.6993, interpMode=RCIM_Cubic, leaveTangent=33.6993, tangentMode=RCTM_User, time=0, value=0}
          <at 0.5 10>
          {arriveTangent=-153.0608, interpMode=RCIM_Cubic, leaveTangent=-153.0608, tangentMode=RCTM_User, time=1, value=0}
         ]}
        ZCurve {keys [<at 0 5>,<at 1 10>]}
       }}
      CurveIndex=$Particles.FrontToBack
      Scale Curve=&Multiply_Vector(
       A=<1 1 1>
       B=&Lerp_Vector(A=<1.5 6 1>, B=<1.5 -6 1>, Alpha=&MakeFloatFromInt(Int=$Particles.SpawnGroup))
      )
     )
    )
    Sampled Position Offset Coordinate Space:ENiagaraMeshTransforms=::NewEnumerator3
    Skeletal Mesh=$User.SkeletalMesh
    Socket Sampling Mode=::NewEnumerator5
   ) {module=SkeletalMeshLocation001}
   &SubUVAnimation(SubUV Animation Mode=::NewEnumerator1)
   /SetVariables(Particles.Velocity=&Multiply_VectorByFloat(
     Vector=&ReturnMeshOrientationAxis(Mesh Orientation Axis=::XAxis, Mesh Orientation Quat=$Particles.MeshOrientation)
     Float=&Lerp_Float(A=-50, B=-50, Alpha=$User.ForwardDot)
    )) {isSetParameters=true}
   &AddVelocity(
    Scale Added Velocity=&Lerp_Vector(
     A=<1 1 1>
     B=<1.1 1.1 1.25>
     Alpha=&Add_Float(A=&ClampFloat(Float=$User.ForwardDot), B=$Particles.LeanLerp)
    )
    Velocity=&RandomRangeVector(Minimum=<0 200 220>, Maximum=<0 450 280>)
   )
   &AddVelocityFromPoint(
    Module.Use Falloff Distance=false
    Origin Offset=&Lerp_Vector(A=<-4000 0 400>, B=<4000 0 100>, Alpha=&MakeFloatFromInt(Int=$Particles.SpawnGroup))
    Origin Offset Coordinate Space=::Simulation
    Velocity Strength=30
   )
  ]
  particleUpdateScript [
   &ParticleState()
   &InheritVelocity(
    Module.Limit Speed=true
    Velocity Limit=10000
    Velocity Scale=&Multiply_VectorByFloat(Vector=<-0.03 -0.03 -0.03>, Float=0.5)
    Velocity Source=&TransformVector(Vector=$Engine.Owner.Velocity, Source Space=::World, Destination Space=::Local)
   ) {enabled=false}
   &SpriteRotationRate(Rotation Rate=&RandomRangeFloat(Minimum=-90, Maximum=90))
   &Drag(Drag=0.5, Use Linear Drag=true, Use Rotational Drag=false)
   &GravityForce(Gravity=<0 0 -900>)
   &SolveForcesAndVelocity(Clamp Velocity=false, Limit Acceleration=false)
   &ScaleSpriteSize(Scale Factor=&Vector2DFromFloat(Value=&FloatFromCurve(
      FloatCurve=DataInterface() {propertyValues {Curve {keys [
          {arriveTangent=1.6115, interpMode=RCIM_Cubic, leaveTangent=1.6115, tangentMode=RCTM_User, time=0, value=0.33}
          {interpMode=RCIM_Cubic, tangentMode=RCTM_User, time=1, value=1}
         ]}}}
      Scale Curve=0.5
     )))
   &DynamicMaterialParameters(
    Index 0 Param 1=&RandomRangeFloat(Minimum=0, Maximum=1)
    Index 0 Param 3=&FloatFromCurve(FloatCurve=DataInterface() {propertyValues {Curve {keys [{interpMode=RCIM_Cubic, tangentMode=RCTM_User, time=0, value=0}, {interpMode=RCIM_Cubic, tangentMode=RCTM_User, time=1, value=1}]}}})
    Index 0 Param 4=&FloatFromCurve(FloatCurve=DataInterface() {propertyValues {Curve {keys [<at 0 0 RCIM_Cubic>, {arriveTangent=1.5276, interpMode=RCIM_Cubic, leaveTangent=1.5276, tangentMode=RCTM_User, time=1, value=1}]}}})
    Module.Param0WriteEnabled=true
    Module.Param1WriteEnabled=false
    Module.Param2WriteEnabled=true
    Module.Param3WriteEnabled=true
   )
  ]
  rendererClasses [&NiagaraSpriteRendererProperties]
  renderers [{rendererClass=&NiagaraSpriteRendererProperties}]
 }
 Wake_Mesh {
  emitterUpdateScript [
   &EmitterState(Delay First Loop Only=true, Inactive Response=::NewEnumerator2, Life Cycle Mode=::NewEnumerator1, Loop Duration=4000, UseLoopDelay=true)
   &SpawnRate(SpawnRate=30, Use Spawn Probability=false)
  ]
  particleSpawnScript [
   /Niagara/Modules/Spawn/Initialization/InitializeParticle(
    Lifetime=1.5
    Mesh Scale=&UniformRangedVector(Minimum=<1.25 1.25 0.1>, Maximum=<2 1.4 0.1>)
    Module.Write Color=true
    Module.Write Lifetime=true
    Module.Write Mass=false
    Module.Write Material Random=true
    Module.Write Position=true
    Module.Write Scale=true
    Module.Write SpriteRotation=false
    Module.Write SpriteSize=false
    Position=&Add_Vector(
     A=/Niagara/DynamicInputs/Helpers/SimulationPosition()
     B=&Multiply_VectorByFloat(Vector=$Engine.Owner.SystemXAxis, Float=-150)
    )
   )
   &SkeletalMeshLocation(
    Apply Sampled Orientation to Mesh Particles=true
    Bone / Socket Index=2
    Inherit Velocity=false
    "Inherit Velocity (Scale)"=-0.2
    Interpolate Mesh Sampling=false
    MaskSpawnGroup=false
    Mesh Sampling Type=::NewEnumerator5
    Sampled Position Offset=&Lerp_Vector(A=<-180 0 0>, B=<-25 0 5>, Alpha=&ClampFloat(Float=$User.ForwardDot, Max=0))
    Sampled Position Offset Coordinate Space:ENiagaraMeshTransforms=::NewEnumerator3
    Skeletal Mesh=$User.SkeletalMesh
    Socket Sampling Mode=::NewEnumerator5
   )
   &InitialMeshOrientation(Rotation Coordinate Space=::NewEnumerator0, Use Orientation Vector=true, Use Rotation Vector=false) {enabled=false, module=InitialMeshOrientation001}
   &AddVelocity(Velocity=<0 0 20>)
   /SetVariables(Particles.Scale=&Multiply_Vector(
     A=$Particles.Scale
     B=&Lerp_Vector(
      A=<0 0 0.1>
      B=<0.8 2 0.1>
      Alpha=&ClampFloat(Float=$System.Timeline.Time, Min=&FN_FrontEndFloatModifier(Gameplay=0, Front End=1))
     )
    )) {isSetParameters=true, module=SetVariables_#5}
  ]
  particleUpdateScript [
   &ParticleState()
   &InheritVelocity(
    Module.Limit Speed=true
    Velocity Limit=10000
    Velocity Scale=&Multiply_VectorByFloat(Vector=<-0.05 -0.05 0>, Float=0.2)
    Velocity Source=&TransformVector(Vector=$Engine.Owner.Velocity, Source Space=::World, Destination Space=::Local)
   ) {enabled=false}
   &InheritSourceMovement(
    Applied Force Scale=<-1 -1 0>
    Applied Force Speed Limit=200
    Applied Position Scale=<0.25 0.25 0>
    ApplyForce=true
    ApplyForceFalloff=false
    ApplyPosition=false
    Module.LimitForceSpeed=false
   ) {enabled=false}
   /SetVariables(Particles.Scale=&Multiply_Vector(
     A=$Particles.Initial.Scale
     B=&VectorFromCurve(
      VectorCurve=DataInterface() {propertyValues {XCurve {keys [<at 0 1>,<at 0.9989 1>]}, YCurve {keys [<at 0 1>,<at 0.9989 1>]}, ZCurve {keys [<at 0 0>,<at 0.2345 0.7983>,<at 0.9989 1>]}}}
      CurveIndex=$Emitter.Scalability.DistanceFraction
     )
    )) {enabled=false, isSetParameters=true, module=SetVariables_#6}
   &Drag(Drag=0, Rotational Drag=20, Use Linear Drag=true, Use Rotational Drag=true)
   &ScaleColor(
    Scale Alpha=&FloatFromCurve(FloatCurve=DataInterface() {propertyValues {Curve {keys [
         {arriveTangent=9.9926, interpMode=RCIM_Cubic, leaveTangent=9.9927, tangentMode=RCTM_User, time=0, value=0}
         {arriveTangent=-2.8339, interpMode=RCIM_Cubic, leaveTangent=-2.8339, tangentMode=RCTM_User, time=0.0216, value=1}
         <at 1.0011 0.002 RCIM_Cubic>
        ]}}})
    ScaleA=true
    ScaleRGB=true
   )
   &SolveForcesAndVelocity(Clamp Velocity=false, Limit Acceleration=false)
   /SetVariables(Particles.Scale=&Multiply_Vector(
     A=$Particles.Initial.Scale
     B=&VectorFromCurve(VectorCurve=DataInterface() {propertyValues {
        XCurve {keys [
          {arriveTangent=4.6994, interpMode=RCIM_Cubic, leaveTangent=4.6994, tangentMode=RCTM_User, time=0, value=0.0795}
          {arriveTangent=0.4589, interpMode=RCIM_Cubic, leaveTangent=0.4589, tangentMode=RCTM_User, time=0.15, value=0.7609}
          <at 1 1>
         ]}
        YCurve {keys [
          {arriveTangent=1.7936, interpMode=RCIM_Cubic, leaveTangent=1.7936, tangentMode=RCTM_User, time=0, value=0}
          <at 0.4 0.4>
          {arriveTangent=-0.1215, interpMode=RCIM_Cubic, leaveTangent=-0.1215, tangentMode=RCTM_User, time=1, value=1}
         ]}
        ZCurve {keys [<at 0 0 RCIM_Cubic>,<at 0.33 1 RCIM_Cubic>,<at 0.9988 0>]}
       }})
    )) {isSetParameters=true, module=SetVariables_#7}
   &DynamicMaterialParameters(
    Index 0 Param 1=&FloatFromCurve(FloatCurve=DataInterface() {propertyValues {Curve {keys [
         {arriveTangent=0.2615, interpMode=RCIM_Cubic, leaveTangent=0.2615, tangentMode=RCTM_User, time=0, value=-1.7984}
         {arriveTangent=3.2109, interpMode=RCIM_Cubic, leaveTangent=3.2109, tangentMode=RCTM_User, time=0.9985, value=0}
        ]}}})
    Index 0 Param 2=&Add_Float(
     A=&FloatFromCurve(FloatCurve=DataInterface() {propertyValues {Curve {keys [
          {arriveTangent=-9.4845, interpMode=RCIM_Cubic, leaveTangent=-9.4845, tangentMode=RCTM_User, time=0, value=4}
          {arriveTangent=0.2383, interpMode=RCIM_Cubic, leaveTangent=0.2383, tangentMode=RCTM_User, time=1, value=0}
         ]}}})
     B=0
    )
    Index 1 Param 1=&FloatFromCurve(FloatCurve=DataInterface() {propertyValues {Curve {keys [
         <at 0 0.1 RCIM_Cubic>
         {arriveTangent=0.9, interpMode=RCIM_Cubic, leaveTangent=0.9, time=0.5861, value=0.2}
         {arriveTangent=3.3294, interpMode=RCIM_Cubic, leaveTangent=3.3294, tangentMode=RCTM_User, time=1, value=1}
        ]}}})
    Module.Param0WriteEnabled=true
    Module.Param0WriteEnabled1=true
    Module.Param1WriteEnabled=true
    Module.Param1WriteEnabled1=true
    Module.Param2WriteEnabled=true
    Module.Param2WriteEnabled1=true
    Module.Param3WriteEnabled=true
    Module.Param3WriteEnabled1=true
    Write Parameter Index 1=true
   )
  ]
  rendererClasses [&NiagaraMeshRendererProperties]
  renderers [{rendererClass=&NiagaraMeshRendererProperties, rendererIndex=0}]
 }
 Wake_Mesh001 {
  emitterUpdateScript [
   &EmitterState(Inactive Response=::NewEnumerator2, Life Cycle Mode=::NewEnumerator1, Loop Behavior=::NewEnumerator1, Loop Duration=0.25, UseLoopDelay=false)
   &SpawnRate(
    SpawnRate=&Lerp_Float(A=30, B=30, Alpha=&ClampFloat(Float=$User.ForwardDot))
    Use Spawn Probability=false
   ) {enabled=false}
   &SpawnBurst_Instantaneous(Use Spawn Probability=false, UseLoopCountLimit=false)
  ]
  particleSpawnScript [
   &InitializeParticle(
    Lifetime=1.5
    Mesh Scale=<0.1 0.1 0.1>
    Module.Write Color=true
    Module.Write Lifetime=true
    Module.Write Mass=false
    Module.Write Material Random=true
    Module.Write Position=true
    Module.Write Scale=true
    Module.Write SpriteRotation=false
    Module.Write SpriteSize=false
   ) {enabled=false}
   /Niagara/Modules/Spawn/Initialization/V2/InitializeParticle(Mesh Scale=<1.1 1 1>, Mesh Scale Mode=::NewEnumerator4, UsePositionOffset=false, Write Material Random=true) {module=InitializeParticle001}
   &InitialMeshOrientation(Rotation Coordinate Space=::NewEnumerator0, Use Orientation Vector=true, Use Rotation Vector=false) {enabled=false, module=InitialMeshOrientation001}
   &AddVelocity()
   /SetVariables(Particles.Scale=&Multiply_Vector(
     A=$Particles.Scale
     B=&Lerp_Vector(A=<0 0 0>, B=<1 1 1>, Alpha=&ClampFloat(Float=$System.Timeline.Time, Min=1))
    )) {isSetParameters=true, module=SetVariables_#5}
   &InitialMeshOrientation(Mesh Orientation Mode=::NewEnumerator1, Rotation=<0 0 0.25>, Use Rotation Vector=true)
  ]
  particleUpdateScript [
   &ParticleState(Kill Particles When Lifetime Has Elapsed=false)
   &SkeletalMeshLocation(
    Apply Sampled Orientation to Mesh Particles=true
    Bone / Socket Index=2
    Inherit Velocity=false
    "Inherit Velocity (Scale)"=-0.2
    Interpolate Mesh Sampling=false
    MaskSpawnGroup=false
    Mesh Sampling Type=::NewEnumerator5
    Sampled Position Offset=&Lerp_Vector(A=<-400 0 15>, B=<-25 0 5>, Alpha=&ClampFloat(Float=$User.ForwardDot, Max=0))
    Sampled Position Offset Coordinate Space:ENiagaraMeshTransforms=::NewEnumerator3
    Skeletal Mesh=$User.SkeletalMesh
    Socket Sampling Mode=::NewEnumerator5
   )
   &InheritVelocity(
    Module.Limit Speed=true
    Velocity Limit=10000
    Velocity Scale=&Multiply_VectorByFloat(Vector=<-0.05 -0.05 0>, Float=0.2)
    Velocity Source=&TransformVector(Vector=$Engine.Owner.Velocity, Source Space=::World, Destination Space=::Local)
   ) {enabled=false}
   &InheritSourceMovement(
    Applied Force Scale=<-1 -1 0>
    Applied Force Speed Limit=200
    Applied Position Scale=<0.25 0.25 0>
    ApplyForce=true
    ApplyForceFalloff=false
    ApplyPosition=false
    Module.LimitForceSpeed=false
   ) {enabled=false}
   /SetVariables(Particles.Scale=&Multiply_Vector(
     A=$Particles.Initial.Scale
     B=&VectorFromCurve(
      VectorCurve=DataInterface() {propertyValues {XCurve {keys [<at 0 1>,<at 0.9989 1>]}, YCurve {keys [<at 0 1>,<at 0.9989 1>]}, ZCurve {keys [<at 0 0>,<at 0.2345 0.7983>,<at 0.9989 1>]}}}
      CurveIndex=$Emitter.Scalability.DistanceFraction
     )
    )) {enabled=false, isSetParameters=true, module=SetVariables_#6}
   &Drag(Drag=0, Rotational Drag=20, Use Linear Drag=true, Use Rotational Drag=true)
   &ScaleColor(
    Scale Alpha=&Lerp_Float(
     A=&FN_FrontEndFloatModifier(
      Gameplay=0
      Front End=&Multiply_Float(
       A=4
       B=&FloatFromCurve(
        FloatCurve=DataInterface() {propertyValues {Curve {keys [
            {arriveTangent=1.8034, interpMode=RCIM_Cubic, leaveTangent=1.8034, tangentMode=RCTM_User, time=0, value=0}
            {arriveTangent=-0.0405, interpMode=RCIM_Cubic, leaveTangent=-0.0405, tangentMode=RCTM_User, time=1, value=1}
            {arriveTangent=0.9995, interpMode=RCIM_Cubic, leaveTangent=0.9995, tangentMode=RCTM_User, time=4, value=0.6825}
           ]}}}
        CurveIndex=$Particles.Age
       )
      )
     )
     B=1
     Alpha=$System.Timeline.Time
    )
    Scale RGB=&VectorFromFloat(Value=1)
    ScaleA=true
    ScaleRGB=true
   )
   &SolveForcesAndVelocity(Clamp Velocity=false, Limit Acceleration=false)
   /SetVariables(Particles.Scale=&Multiply_Vector(
     A=$Particles.Initial.Scale
     B=&VectorFromCurve(VectorCurve=DataInterface() {propertyValues {
        XCurve {keys [
          {arriveTangent=4.6994, interpMode=RCIM_Cubic, leaveTangent=4.6994, tangentMode=RCTM_User, time=0, value=0.0795}
          {arriveTangent=0.4589, interpMode=RCIM_Cubic, leaveTangent=0.4589, tangentMode=RCTM_User, time=0.15, value=0.7609}
          <at 1 1>
         ]}
        YCurve {keys [{arriveTangent=2.5596, interpMode=RCIM_Cubic, leaveTangent=2.5596, tangentMode=RCTM_User, time=0, value=0.0767}, <at 1 1>]}
        ZCurve {keys [<at 0 0 RCIM_Cubic>,<at 0.33 1 RCIM_Cubic>,<at 0.9988 0>]}
       }})
    )) {enabled=false, isSetParameters=true, module=SetVariables_#7}
   &DynamicMaterialParameters(
    Index 0 Param 1=&FloatFromCurve(FloatCurve=DataInterface() {propertyValues {Curve {keys [
         {arriveTangent=0.2615, interpMode=RCIM_Cubic, leaveTangent=0.2615, tangentMode=RCTM_User, time=0, value=-1.7984}
         {arriveTangent=3.2109, interpMode=RCIM_Cubic, leaveTangent=3.2109, tangentMode=RCTM_User, time=0.9985, value=0}
        ]}}})
    Index 0 Param 2=&Add_Float(
     A=&FloatFromCurve(FloatCurve=DataInterface() {propertyValues {Curve {keys [
          {arriveTangent=-9.4845, interpMode=RCIM_Cubic, leaveTangent=-9.4845, tangentMode=RCTM_User, time=0, value=4}
          {arriveTangent=0.2383, interpMode=RCIM_Cubic, leaveTangent=0.2383, tangentMode=RCTM_User, time=1, value=0}
         ]}}})
     B=0
    )
    Index 1 Param 1=&FloatFromCurve(FloatCurve=DataInterface() {propertyValues {Curve {keys [
         <at 0 0.1 RCIM_Cubic>
         {arriveTangent=0.9, interpMode=RCIM_Cubic, leaveTangent=0.9, time=0.5861, value=0.2}
         {arriveTangent=3.3294, interpMode=RCIM_Cubic, leaveTangent=3.3294, tangentMode=RCTM_User, time=1, value=1}
        ]}}})
    Module.Param0WriteEnabled=true
    Module.Param0WriteEnabled1=true
    Module.Param1WriteEnabled=true
    Module.Param1WriteEnabled1=true
    Module.Param2WriteEnabled=true
    Module.Param2WriteEnabled1=true
    Module.Param3WriteEnabled=true
    Module.Param3WriteEnabled1=true
    Write Parameter Index 1=true
   )
  ]
  rendererClasses [&NiagaraMeshRendererProperties]
  renderers [{rendererClass=&NiagaraMeshRendererProperties, rendererIndex=0}]
 }
 Wake_Mesh002 {
  bEnabled=false
  emitterUpdateScript [
   &EmitterState(Delay First Loop Only=true, Inactive Response=::NewEnumerator2, Life Cycle Mode=::NewEnumerator1, Loop Duration=4000, UseLoopDelay=true)
   &SpawnRate(SpawnRate=30, Use Spawn Probability=false)
  ]
  particleSpawnScript [
   /Niagara/Modules/Spawn/Initialization/InitializeParticle(
    Lifetime=1.5
    Mesh Scale=&UniformRangedVector(Minimum=<1.25 1.25 0.8>, Maximum=<2 1.4 1.15>)
    Module.Write Color=true
    Module.Write Lifetime=true
    Module.Write Mass=false
    Module.Write Material Random=true
    Module.Write Position=true
    Module.Write Scale=true
    Module.Write SpriteRotation=false
    Module.Write SpriteSize=false
    Position=&Add_Vector(
     A=&SimulationPosition()
     B=&Multiply_VectorByFloat(Vector=$Engine.Owner.SystemXAxis, Float=-150)
    )
   )
   &SkeletalMeshLocation(
    Apply Sampled Orientation to Mesh Particles=true
    Bone / Socket Index=2
    Inherit Velocity=false
    "Inherit Velocity (Scale)"=-0.2
    Interpolate Mesh Sampling=false
    MaskSpawnGroup=false
    Mesh Sampling Type=::NewEnumerator5
    Sampled Position Offset=&Lerp_Vector(A=<-180 0 0>, B=<-25 0 5>, Alpha=&ClampFloat(Float=$User.ForwardDot, Max=0))
    Sampled Position Offset Coordinate Space:ENiagaraMeshTransforms=::NewEnumerator3
    Skeletal Mesh=$User.SkeletalMesh
    Socket Sampling Mode=::NewEnumerator5
   )
   &InitialMeshOrientation(Rotation Coordinate Space=::NewEnumerator0, Use Orientation Vector=true, Use Rotation Vector=false) {enabled=false, module=InitialMeshOrientation001}
   &AddVelocity(Velocity=<0 0 20>)
   /SetVariables(Particles.Scale=&Multiply_Vector(
     A=$Particles.Scale
     B=&TransformVector(
      Vector=&Lerp_Vector(
       A=<0 0 0>
       B=<0.8 2 0>
       Alpha=&ClampFloat(Float=$System.Timeline.Time, Min=&FN_FrontEndFloatModifier(Gameplay=0, Front End=1))
      )
      Destination Space=::World
     )
    )) {enabled=false, isSetParameters=true, module=SetVariables_#5}
  ]
  particleUpdateScript [
   &ParticleState()
   &InheritVelocity(
    Module.Limit Speed=true
    Velocity Limit=10000
    Velocity Scale=&Multiply_VectorByFloat(Vector=<-0.05 -0.05 0>, Float=0.2)
    Velocity Source=&TransformVector(Vector=$Engine.Owner.Velocity, Source Space=::World, Destination Space=::Local)
   ) {enabled=false}
   &InheritSourceMovement(
    Applied Force Scale=<-1 -1 0>
    Applied Force Speed Limit=200
    Applied Position Scale=<0.25 0.25 0>
    ApplyForce=true
    ApplyForceFalloff=false
    ApplyPosition=false
    Module.LimitForceSpeed=false
   ) {enabled=false}
   /SetVariables(Particles.Scale=&Multiply_Vector(
     A=$Particles.Initial.Scale
     B=&VectorFromCurve(
      VectorCurve=DataInterface() {propertyValues {XCurve {keys [<at 0 1>,<at 0.9989 1>]}, YCurve {keys [<at 0 1>,<at 0.9989 1>]}, ZCurve {keys [<at 0 0>,<at 0.2345 0.7983>,<at 0.9989 1>]}}}
      CurveIndex=$Emitter.Scalability.DistanceFraction
     )
    )) {enabled=false, isSetParameters=true, module=SetVariables_#6}
   &Drag(Drag=0, Rotational Drag=20, Use Linear Drag=true, Use Rotational Drag=true)
   &ScaleColor(
    Scale Alpha=&FloatFromCurve(FloatCurve=DataInterface() {propertyValues {Curve {keys [
         {arriveTangent=9.9926, interpMode=RCIM_Cubic, leaveTangent=9.9927, tangentMode=RCTM_User, time=0, value=0}
         {arriveTangent=-2.8339, interpMode=RCIM_Cubic, leaveTangent=-2.8339, tangentMode=RCTM_User, time=0.0216, value=1}
         <at 1.0011 0.002 RCIM_Cubic>
        ]}}})
    ScaleA=true
    ScaleRGB=true
   )
   &SolveForcesAndVelocity(Clamp Velocity=false, Limit Acceleration=false)
   /SetVariables(Particles.Scale=&Multiply_Vector(
     A=$Particles.Initial.Scale
     B=&VectorFromCurve(VectorCurve=DataInterface() {propertyValues {
        XCurve {keys [
          {arriveTangent=4.6994, interpMode=RCIM_Cubic, leaveTangent=4.6994, tangentMode=RCTM_User, time=0, value=0.0795}
          {arriveTangent=0.4589, interpMode=RCIM_Cubic, leaveTangent=0.4589, tangentMode=RCTM_User, time=0.15, value=0.7609}
          <at 1 1>
         ]}
        YCurve {keys [
          {arriveTangent=1.7936, interpMode=RCIM_Cubic, leaveTangent=1.7936, tangentMode=RCTM_User, time=0, value=0}
          <at 0.4 0.4>
          {arriveTangent=-0.1215, interpMode=RCIM_Cubic, leaveTangent=-0.1215, tangentMode=RCTM_User, time=1, value=1}
         ]}
        ZCurve {keys [<at 0 0 RCIM_Cubic>,<at 0.33 1 RCIM_Cubic>,<at 0.9988 0>]}
       }})
    )) {isSetParameters=true, module=SetVariables_#7}
   &DynamicMaterialParameters(
    Index 0 Param 1=&FloatFromCurve(FloatCurve=DataInterface() {propertyValues {Curve {keys [
         {arriveTangent=0.2615, interpMode=RCIM_Cubic, leaveTangent=0.2615, tangentMode=RCTM_User, time=0, value=-1.7984}
         {arriveTangent=3.2109, interpMode=RCIM_Cubic, leaveTangent=3.2109, tangentMode=RCTM_User, time=0.9985, value=0}
        ]}}})
    Index 0 Param 2=&Add_Float(
     A=&FloatFromCurve(FloatCurve=DataInterface() {propertyValues {Curve {keys [
          {arriveTangent=-9.4845, interpMode=RCIM_Cubic, leaveTangent=-9.4845, tangentMode=RCTM_User, time=0, value=4}
          {arriveTangent=0.2383, interpMode=RCIM_Cubic, leaveTangent=0.2383, tangentMode=RCTM_User, time=1, value=0}
         ]}}})
     B=0
    )
    Index 1 Param 1=&FloatFromCurve(FloatCurve=DataInterface() {propertyValues {Curve {keys [
         <at 0 0.1 RCIM_Cubic>
         {arriveTangent=0.9, interpMode=RCIM_Cubic, leaveTangent=0.9, time=0.5861, value=0.2}
         {arriveTangent=3.3294, interpMode=RCIM_Cubic, leaveTangent=3.3294, tangentMode=RCTM_User, time=1, value=1}
        ]}}})
    Module.Param0WriteEnabled=true
    Module.Param0WriteEnabled1=true
    Module.Param1WriteEnabled=true
    Module.Param1WriteEnabled1=true
    Module.Param2WriteEnabled=true
    Module.Param2WriteEnabled1=true
    Module.Param3WriteEnabled=true
    Module.Param3WriteEnabled1=true
    Write Parameter Index 1=true
   )
  ]
  rendererClasses [&NiagaraMeshRendererProperties]
  renderers [
   {rendererClass=&NiagaraMeshRendererProperties, rendererIndex=0}
   {rendererClass=&NiagaraMeshRendererProperties, rendererIndex=1}
  ]
 }
}
summary {
 systemName=NS_Glider_Journey
 emitters [
  {emitterName=Splash_Drops, rendererClasses [&NiagaraSpriteRendererProperties]}
  {emitterName=Splash_Foam, rendererClasses [&NiagaraSpriteRendererProperties]}
  {emitterName=Wake_Mesh, rendererClasses [&NiagaraMeshRendererProperties]}
  {emitterName=Edge_Spray_Mesh, rendererClasses [&NiagaraMeshRendererProperties]}
  {bEnabled=false, emitterName=FN_Water_Foam_SnapToSurface001, rendererClasses [&NiagaraMeshRendererProperties]}
  {bEnabled=false, emitterName=FN_Water_Foam_SnapToSurface_Detail, rendererClasses [&NiagaraMeshRendererProperties]}
  {bEnabled=false, emitterName=Smokes001, rendererClasses [&NiagaraSpriteRendererProperties]}
  {bEnabled=false, emitterName=Speedlines, rendererClasses [&NiagaraSpriteRendererProperties]}
  {emitterName=Speedlines001, rendererClasses [&NiagaraSpriteRendererProperties]}
  {emitterName=Wake_Mesh001, rendererClasses [&NiagaraMeshRendererProperties]}
  {bEnabled=false, emitterName=Wake_Mesh002, rendererClasses [&NiagaraMeshRendererProperties]}
  {bEnabled=false, emitterName=Edge_Spray_Mesh001, rendererClasses [&NiagaraMeshRendererProperties]}
 ]
 userVariables [
  User.bIsFrontEnd {type=NiagaraBool}
  User.bIsFrontEndPreview {type=NiagaraBool}
  User.bIsFullyDeployed {type=NiagaraBool}
  User.ForwardDot {type=NiagaraFloat, defaultValue=0, description=""}
  User.RightDot {type=NiagaraFloat, defaultValue=0, description=""}
  User.SkeletalMesh {
   type=NiagaraDataInterfaceSkeletalMesh
   defaultValue {
    struct=NiagaraExt_VariableValue_DataInterface
    value {dataInterface=&self:NiagaraDataInterfaceSkeletalMesh_1}
   }
  }
  User.WaterBodyIndex {type=NiagaraFloat, defaultValue=0, description=""}
  User.WaterDepth {type=NiagaraFloat, defaultValue=0, description=""}
  User.WaterPlaneLocation {type=Vector3f, description="", defaultValue {struct=Vector3f, value=<0 0 0>}}
  User.WaterPlaneNormal {type=Vector3f, description="", defaultValue {struct=Vector3f, value=<0 0 0>}}
  User.WaveAttenuationFactor {type=NiagaraFloat, defaultValue=0, description=""}
  User.WaveTime {type=NiagaraFloat, defaultValue=0, description=""}
 ]
}
