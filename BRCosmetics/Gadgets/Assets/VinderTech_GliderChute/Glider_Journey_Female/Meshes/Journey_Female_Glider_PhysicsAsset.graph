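# Physics asset for the Journey Female glider cosmetic, giving its hull, sails, flag and handles capsule collision bodies.
# The four sail bodies constrain directly to the large hull capsule, while the flag chains from the hull through joints 01, 03 and 05.
# Left and right handles constrain to a handle base that has no constraint to the hull.
# All constraints share the same swing and twist limit angles.
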PhysicsAsset /BRCosmetics/Gadgets/Assets/VinderTech_GliderChute/Glider_Journey_Female/Meshes/Journey_Female_Glider_PhysicsAsset
bodies {
 Handle_Base_l [{shapeType=Capsule, length=22.2623, radius=8.4401, center=<-18.5208 -1.5929 -0.7848>, rotation=<rot 0 93.4935 -88.3575>}]
 Handle_Base_r [{shapeType=Capsule, length=22.2623, radius=8.44, center=<18.5207 1.5943 0.7839>, rotation=<rot 0 93.4977 -88.3604>}]
 Handles_Base [{shapeType=Capsule, length=20.0924, radius=12.4509, center=<-0.5349 -1.7923 -2.0202>, rotation=<rot 0 -89.836 -86.9364>}]
 JNT_Flag_01 [{shapeType=Capsule, length=12.882, radius=11.1411, center=<0 10.3241 -0.3037>, rotation=<rot 18.7251 -90 0>}]
 JNT_Flag_03 [{shapeType=Capsule, length=16.4892, radius=10.8809, center=<0 9.7769 -1.1493>, rotation=<rot 0 0 -82.1697>}]
 JNT_Flag_05 [{shapeType=Capsule, length=18.7261, radius=9.9997, center=<0 14.5913 -1.883>, rotation=<rot 0 -180 -80.895>}]
 JNT_Hull_Base [{shapeType=Capsule, length=273.0337, radius=203.3292, center=<0.2715 -57.3095 193.1356>, rotation=<rot 0 -0.0403 -87.6176>}]
 JNT_Sail_lowLeft [{shapeType=Capsule, length=73.2304, radius=41.3681, center=<17.4986 -29.34 29.7162>, rotation=<rot 0 -113.571 -53.1011>}]
 JNT_Sail_lowRight [{shapeType=Capsule, length=72.8325, radius=41.7699, center=<-17.6275 -29.4754 30.0519>, rotation=<rot 0 113.8871 -53.8026>}]
 JNT_Sail_middle [{shapeType=Capsule, length=83.0667, radius=61.1904, center=<0.4564 12.7464 -9.9818>, rotation=<rot 0 90.0369 -89.0889>}]
 JNT_Sail_top [{shapeType=Capsule, length=175.0399, radius=58.3464, center=<0.5991 -30.214 -51.1479>, rotation=<rot 0 90.0387 -89.8996>}]
}
constraints [
 {bone1Name=JNT_Sail_lowLeft, bone2Name=JNT_Hull_Base, swing1LimitDegrees=45, swing2LimitDegrees=45, twistLimitDegrees=45}
 {bone1Name=JNT_Sail_lowRight, bone2Name=JNT_Hull_Base, swing1LimitDegrees=45, swing2LimitDegrees=45, twistLimitDegrees=45}
 {bone1Name=JNT_Sail_middle, bone2Name=JNT_Hull_Base, swing1LimitDegrees=45, swing2LimitDegrees=45, twistLimitDegrees=45}
 {bone1Name=JNT_Sail_top, bone2Name=JNT_Hull_Base, swing1LimitDegrees=45, swing2LimitDegrees=45, twistLimitDegrees=45}
 {bone1Name=JNT_Flag_01, bone2Name=JNT_Hull_Base, swing1LimitDegrees=45, swing2LimitDegrees=45, twistLimitDegrees=45}
 {bone1Name=JNT_Flag_03, bone2Name=JNT_Flag_01, swing1LimitDegrees=45, swing2LimitDegrees=45, twistLimitDegrees=45}
 {bone1Name=JNT_Flag_05, bone2Name=JNT_Flag_03, swing1LimitDegrees=45, swing2LimitDegrees=45, twistLimitDegrees=45}
 {bone1Name=Handle_Base_r, bone2Name=Handles_Base, swing1LimitDegrees=45, swing2LimitDegrees=45, twistLimitDegrees=45}
 {bone1Name=Handle_Base_l, bone2Name=Handles_Base, swing1LimitDegrees=45, swing2LimitDegrees=45, twistLimitDegrees=45}
]
